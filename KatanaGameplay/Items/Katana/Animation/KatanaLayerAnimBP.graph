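# Katana item anim layer for the Fortnite player skeleton that overrides the final pose during dash moves.
# A state machine runs charge (aim offset), into, attack and ground/air outro, with the outro timed from dash root-motion duration.
# Its output gets a pelvis aim offset, inertialization and a hand IK layer that uses Leg IK on hands and pelvis-based virtual bones.
# Other item layers contain no custom logic.
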AnimBlueprint /KatanaGameplay/Items/Katana/Animation/KatanaLayerAnimBP
target_skeleton=&Fortnite_M_Avg_Player_Skeleton
graphs [
 AnimGraph {nodes [Output Pose() {layerGroup=None}]}
 EventGraph {nodes [
   n1 = SetDashChargePlayRate!() -> (OutputDelegate, then, DashChargePlayRate:double=~) {is=title K2Node_CustomEvent}
   > Set DashChargePlayRate(DashChargePlayRate:float=@n1.DashChargePlayRate) -> (Output_Get:float)
   StartDashOutro() {is=title K2Node_CustomEvent}
   > Set StartedDashOutro(StartedDashOutro:bool=true) -> (Output_Get:bool)
   > Clear and Invalidate Timer by Handle(Handle=Get DashOutroTimerHandle() -> (DashOutroTimerHandle:Timer Handle Structure))
   Event BlueprintUpdateAnimation()
   > Branch(Condition=Get IsBMeleeKatana() -> (IsBMeleeKatana:bool))
   > n2 = Sequence()
   n3 = Is Valid Timer Handle!(Handle:Timer Handle Structure=Get DashOutroTimerHandle()) -> (ReturnValue:bool)
   n4 = Get StartedDashOutro() -> (StartedDashOutro:bool)
   > Branch(Condition=AND Boolean!(A=Get IsDashAttack() -> (IsDashAttack:bool), B=NOT Boolean(A=@n3), C:bool=NOT Boolean(A=@n4)) -> (ReturnValue))
   > n5 = Set Timer by Event(
    Delegate=Create Event()
    Time=GetValueAtLevel!(Input:"Scalable Float Structure (by ref)"=Get DashRootMotionDuration() -> (DashRootMotionDuration:Scalable Float Structure), Level:float=0) -> (ReturnValue:float)
    bLooping=false
   )
   > Set DashOutroTimerHandle(DashOutroTimerHandle:Timer Handle Structure=@n5) -> (Output_Get:Timer Handle Structure)
   n6 = Branch(execute=@n2.then_1, Condition=Get IsDashAttack())
   Set StartedDashOutro(execute=@n6.else, StartedDashOutro=false)
   > Clear and Invalidate Timer by Handle(Handle=Get DashOutroTimerHandle())
   Event BlueprintInitializeAnimation()
   n7 = Get AsFortPlayerPawn() -> (AsFortPlayerPawn:Fort Player Pawn Object Reference)
   > Is Valid(InputObject=@n7)
   n8 = GetCurrentWeapon!(self:Fort Pawn Object Reference=@n7) -> (ReturnValue:Fort Weapon Object Reference)
   > Is Valid(InputObject=@n8)
   > Cast To B_Melee_Katana!(execute, Object=@n8) -> (then, CastFailed, AsB Melee Katana:B Melee Katana Object Reference) {is=title K2Node_DynamicCast}
   > Set IsBMeleeKatana(IsBMeleeKatana:bool=true) -> (Output_Get:bool)
  ]}
 HandIKGroup {nodes [
   n1 = Leg IK!(ComponentPose:Component Space Pose Link Structure=Local To Component(LocalPose=Input Pose() {name=InPoseHandIK}), Alpha:float=1) -> (Pose:Component Space Pose Link Structure) {
    is=title AnimGraphNode_LegIK
    alphaBoolBlend {blendOption=Linear}
    legsDefinition [{
      bEnableKneeTwistCorrection=false
      bEnableRotationLimit=false
      footBoneForwardAxis=Y
      hingeRotationAxis=Z
      minRotationAngle=15
      numBonesInLimb=2
      twistOffsetCurveName=None
      fKFootBone {boneName=hand_r}
      iKFootBone {boneName=ik_hand_r}
     }]
   }
   n2 = "Transform (Modify) Bone - Bone: VB pelvis_hand_attach_r"!(
    ComponentPose=Copy Bone - Source Bone: pelvis - Target Bone: VB pelvis_hand_attach_r!(
     ComponentPose:Component Space Pose Link Structure=@n1
     Alpha:float=1
     bCopyTranslation:bool=true
     bCopyRotation:bool=true
     bCopyScale:bool=False
    ) -> (Pose:Component Space Pose Link Structure) {
     is=title AnimGraphNode_CopyBone
     alphaBoolBlend {blendOption=Linear}
     sourceBone {boneName=pelvis}
     targetBone {boneName=VB pelvis_hand_attach_r}
    }
    Alpha=1
    Translation="0,0,0"
    Rotation="0,0,0"
   ) -> (Pose) {
    rotationMode=BMM_Additive
    rotationSpace=BCS_ParentBoneSpace
    translationMode=BMM_Additive
    translationSpace=BCS_ParentBoneSpace
    alphaBoolBlend {blendOption=Linear}
    boneToModify {boneName=VB pelvis_hand_attach_r}
   }
   n3 = "Transform (Modify) Bone - Bone: VB pelvis_hand_attach_r"!(ComponentPose=@n2, bAlphaBoolEnabled:bool=True, Translation="-10,0,0", Rotation="0,40,-12") -> (Pose) {
    alphaInputType=Bool
    rotationMode=BMM_Additive
    rotationSpace=BCS_ParentBoneSpace
    translationMode=BMM_Additive
    translationSpace=BCS_ParentBoneSpace
    alphaBoolBlend {blendOption=Linear}
    boneToModify {boneName=VB pelvis_hand_attach_r}
    rotation {roll=-12, yaw=40}
    translation {x=-10}
   }
   n4 = Copy Bone - Source Bone: VB pelvis_hand_attach_r - Target Bone: hand_attach_r!(
    ComponentPose:Component Space Pose Link Structure=@n3
    Alpha:float=1
    bCopyTranslation:bool=true
    bCopyRotation:bool=true
    bCopyScale:bool=False
   ) -> (Pose:Component Space Pose Link Structure) {
    is=title AnimGraphNode_CopyBone
    alphaBoolBlend {blendOption=Linear}
    sourceBone {boneName=VB pelvis_hand_attach_r}
    targetBone {boneName=hand_attach_r}
   }
   n5 = "Transform (Modify) Bone - Bone: VB pelvis_ik_hand_l"!(
    ComponentPose=Copy Bone - Source Bone: pelvis - Target Bone: VB pelvis_ik_hand_l!(
     ComponentPose:Component Space Pose Link Structure=@n4
     Alpha:float=1
     bCopyTranslation:bool=true
     bCopyRotation:bool=true
     bCopyScale:bool=False
    ) -> (Pose:Component Space Pose Link Structure) {
     is=title AnimGraphNode_CopyBone
     alphaBoolBlend {blendOption=Linear}
     sourceBone {boneName=pelvis}
     targetBone {boneName=VB pelvis_ik_hand_l}
    }
    Alpha=1
    Translation="0,0,0"
    Rotation="0,0,0"
   ) -> (Pose) {
    rotationMode=BMM_Additive
    rotationSpace=BCS_ParentBoneSpace
    translationMode=BMM_Additive
    translationSpace=BCS_ParentBoneSpace
    alphaBoolBlend {blendOption=Linear}
    boneToModify {boneName=VB pelvis_ik_hand_l}
   }
   n6 = "Transform (Modify) Bone - Bone: VB pelvis_ik_hand_l"!(ComponentPose=@n5, bAlphaBoolEnabled:bool=True, Translation="7,0,0", Rotation="0,0,0") -> (Pose) {
    alphaInputType=Bool
    rotationMode=BMM_Additive
    rotationSpace=BCS_ParentBoneSpace
    translationMode=BMM_Additive
    translationSpace=BCS_ParentBoneSpace
    alphaBoolBlend {blendOption=Linear}
    boneToModify {boneName=VB pelvis_ik_hand_l}
    translation {x=7}
   }
   n7 = Leg IK(
    ComponentPose=Copy Bone - Source Bone: VB pelvis_ik_hand_l - Target Bone: ik_hand_l!(
     ComponentPose:Component Space Pose Link Structure=@n6
     Alpha:float=1
     bCopyTranslation:bool=true
     bCopyRotation:bool=true
     bCopyScale:bool=False
    ) -> (Pose:Component Space Pose Link Structure) {
     is=title AnimGraphNode_CopyBone
     alphaBoolBlend {blendOption=Linear}
     sourceBone {boneName=VB pelvis_ik_hand_l}
     targetBone {boneName=ik_hand_l}
    }
    Alpha=1
   ) {
    alphaBoolBlend {blendOption=Linear}
    legsDefinition [{
      bEnableKneeTwistCorrection=false
      bEnableRotationLimit=false
      footBoneForwardAxis=Y
      hingeRotationAxis=Z
      minRotationAngle=15
      numBonesInLimb=2
      twistOffsetCurveName=None
      fKFootBone {boneName=hand_l}
      iKFootBone {boneName=ik_hand_l}
     }]
   }
   Output Pose(Result=Component To Local(ComponentPose=@n7)) {layerGroup=None}
  ]}
 ItemAimOffset {nodes [
   Input Pose!() -> (Pose, AimOffsetAlpha:double, Yaw:double, Pitch:double) {name=InPose_AimOffset}
   Input Pose!() -> (Pose) {name=InPose_PreAimOffsets}
   Output Pose() {layerGroup=Item Override}
  ]}
 ItemFallAdditive {nodes [
   Input Pose() {name=InPoseItemFallAdditive}
   Output Pose() {layerGroup=Item Override}
  ]}
 ItemFallLandAdditive {nodes [
   Input Pose() {name=InPoseItemFallLandAdditive}
   Output Pose() {layerGroup=Item Override}
  ]}
 ItemFallLevel2Override {nodes [
   Input Pose() {name=InPose_FallLevel2}
   Output Pose() {layerGroup=Item Override}
  ]}
 ItemFinalPoseOverride {nodes [
   AnimGraphNode_StateMachine_0 = DashStateMachine() {is=title AnimGraphNode_StateMachine, bAllowConduitEntryStates=true}
   n1 = AimOffset Player 'Katana_DashPelvis_AO'!(Y:float=0, BasePose:Pose Link Structure=@AnimGraphNode_StateMachine_0, Alpha:float=1) -> (Pose:Pose Link Structure) {blendSpace=/KatanaGameplay/Items/Katana/Animation/Dash/AO/Katana_DashPelvis_AO, alphaBoolBlend {blendOption=Linear}}
   n2 = Use cached pose 'CachedFinalPose'()
   n3 = HandIKGroup!(InPoseHandIK:Pose Link Structure=Inertialization!(Source:Pose Link Structure=Blend Poses by bool(BlendPose_0=@n2, BlendPose_1=@n1, BlendTime_0=0, BlendTime_1=0) {blendType=Linear}) -> (Pose:Pose Link Structure) {is=title AnimGraphNode_Inertialization}) -> (Pose:Pose Link Structure) {is=title AnimGraphNode_LinkedAnimLayer, layer=HandIKGroup}
   Output Pose(Result=Blend Poses by bool(
     BlendPose_0=Use cached pose 'CachedFinalPose'()
     BlendPose_1=Modify Curve!(SourcePose=@n3, CurveValues_0=1, CurveValues_1:float=0, Alpha=1) -> (Pose) {curveNames [LeftHandIK,LeftHandFK], curveValues [0,0]}
     BlendTime_0=0.5
    ) {blendType=Linear}) {layerGroup=Item Override}
   Save cached pose 'CachedFinalPose'(Pose=Input Pose() {name=InFinalPose}) {cacheName=CachedFinalPose}
  ]}
 DashStateMachine {
  entry=Default
  path=ItemFinalPoseOverride.AnimGraphNode_StateMachine_0.DashStateMachine
  ids=id
  nodes [
   AnimStateNode_0 = Dash Charge() {
    is=title AnimStateNode
    stateEntered {guid=#1}
    stateFullyBlended {guid=#2}
    stateLeft {guid=#3}
   }
   AnimStateNode_1 = Dash Attack() {is=title AnimStateNode}
   AnimStateNode_3 = Dash Attack Into() {is=title AnimStateNode}
   AnimStateNode_4 = Default() {is=title AnimStateNode}
   AnimStateNode_5 = Dash Outro() {
    is=title AnimStateNode
    stateEntered {guid=#4}
    stateFullyBlended {guid=#5}
    stateLeft {guid=#6}
   }
   AnimStateTransitionNode_0 = AnimStateTransitionNode() {crossfadeDuration=0, logicType=TLT_Inertialization}
   AnimStateTransitionNode_1 = AnimStateTransitionNode() {crossfadeDuration=0}
   AnimStateTransitionNode_10 = AnimStateTransitionNode()
   AnimStateTransitionNode_16 = AnimStateTransitionNode() {crossfadeDuration=0, priorityOrder=2}
   AnimStateTransitionNode_2 = AnimStateTransitionNode() {crossfadeDuration=0}
   AnimStateTransitionNode_3 = AnimStateTransitionNode() {bAutomaticRuleBasedOnSequencePlayerInState=true, priorityOrder=2}
   AnimStateTransitionNode_4 = AnimStateTransitionNode() {crossfadeDuration=0.5}
   AnimStateTransitionNode_5 = AnimStateTransitionNode()
   AnimStateTransitionNode_6 = AnimStateTransitionNode() {crossfadeDuration=0, logicType=TLT_Inertialization}
   AnimStateTransitionNode_7 = AnimStateTransitionNode()
   AnimStateTransitionNode_8 = AnimStateTransitionNode() {bAutomaticRuleBasedOnSequencePlayerInState=true, crossfadeDuration=0, priorityOrder=2}
   AnimStateTransitionNode_9 = AnimStateTransitionNode() {crossfadeDuration=0.3, logicType=TLT_Inertialization, priorityOrder=0}
  ]
  links [
   {
    dst=Dash Outro
    rule=ItemFinalPoseOverride.AnimGraphNode_StateMachine_0.DashStateMachine.AnimStateTransitionNode_1.Transition
    src=Dash Attack
    via=AnimStateTransitionNode_1
   }
   {
    dst=Dash Outro
    rule=ItemFinalPoseOverride.AnimGraphNode_StateMachine_0.DashStateMachine.AnimStateTransitionNode_16.Transition
    src=Dash Attack
    via=AnimStateTransitionNode_16
   }
   {
    dst=Dash Attack
    rule=ItemFinalPoseOverride.AnimGraphNode_StateMachine_0.DashStateMachine.AnimStateTransitionNode_5.Transition
    src=Dash Attack Into
    via=AnimStateTransitionNode_5
   }
   {
    dst=Dash Attack
    rule=ItemFinalPoseOverride.AnimGraphNode_StateMachine_0.DashStateMachine.AnimStateTransitionNode_8.Transition
    src=Dash Attack Into
    via=AnimStateTransitionNode_8
   }
   {
    dst=Dash Attack Into
    rule=ItemFinalPoseOverride.AnimGraphNode_StateMachine_0.DashStateMachine.AnimStateTransitionNode_2.Transition
    src=Dash Charge
    via=AnimStateTransitionNode_2
   }
   {
    dst=Default
    rule=ItemFinalPoseOverride.AnimGraphNode_StateMachine_0.DashStateMachine.AnimStateTransitionNode_10.Transition
    src=Dash Charge
    via=AnimStateTransitionNode_10
   }
   {
    dst=Default
    rule=ItemFinalPoseOverride.AnimGraphNode_StateMachine_0.DashStateMachine.AnimStateTransitionNode_7.Transition
    src=Dash Charge
    via=AnimStateTransitionNode_7
   }
   {
    dst=Dash Charge
    rule=ItemFinalPoseOverride.AnimGraphNode_StateMachine_0.DashStateMachine.AnimStateTransitionNode_6.Transition
    src=Dash Outro
    via=AnimStateTransitionNode_6
   }
   {
    dst=Default
    rule=ItemFinalPoseOverride.AnimGraphNode_StateMachine_0.DashStateMachine.AnimStateTransitionNode_3.Transition
    src=Dash Outro
    via=AnimStateTransitionNode_3
   }
   {
    dst=Default
    rule=ItemFinalPoseOverride.AnimGraphNode_StateMachine_0.DashStateMachine.AnimStateTransitionNode_4.Transition
    src=Dash Outro
    via=AnimStateTransitionNode_4
   }
   {
    dst=Default
    rule=ItemFinalPoseOverride.AnimGraphNode_StateMachine_0.DashStateMachine.AnimStateTransitionNode_9.Transition
    src=Dash Outro
    via=AnimStateTransitionNode_9
   }
   {
    dst=Dash Charge
    rule=ItemFinalPoseOverride.AnimGraphNode_StateMachine_0.DashStateMachine.AnimStateTransitionNode_0.Transition
    src=Default
    via=AnimStateTransitionNode_0
   }
  ]
 }
 Dash Charge {
  path=ItemFinalPoseOverride.AnimGraphNode_StateMachine_0.DashStateMachine.AnimStateNode_0.Dash Charge
  nodes [
   n1 = Sequence Player 'Katana_Dash_ChargeUp_Ground'!(PlayRate:float=1) -> (Pose) {
    bLoopAnimation=false
    groupName=Charge
    groupRole=AlwaysLeader
    method=SyncGroup
    sequence=/KatanaGameplay/Items/Katana/Animation/Dash/Katana_Dash_ChargeUp_Ground
   }
   n2 = "Sequence Player (additive) 'Katana_Dash_ChargeUp_InAir'"!(PlayRateBasis:float=1) -> (Pose) {
    bLoopAnimation=false
    groupName=Charge
    groupRole=AlwaysFollower
    method=SyncGroup
    sequence=/KatanaGameplay/Items/Katana/Animation/Dash/Katana_Dash_ChargeUp_InAir
   }
   Output Animation Pose(Result=AimOffset Player 'Katana_Dash_AO'!(
     X:float=Get YawForKatanaDashAO() -> (YawForKatanaDashAO:float)
     Y:float=Get!() -> (AdjustedAim_ResultingYaw:float=~, AdjustedAim_ResultingPitch:float=~, AdjustedAim_ZeroOutPitchWeightCurveName:Name).AdjustedAim_ResultingPitch
     BasePose:Pose Link Structure=Apply Additive(Base=@n1, Additive=@n2) {alphaBoolBlend {blendOption=Linear}}
     Alpha:float=1
    ) -> (Pose:Pose Link Structure) {blendSpace=/KatanaGameplay/Items/Katana/Animation/Dash/AO/Katana_Dash_AO, alphaBoolBlend {blendOption=Linear}})
  ]
 }
 Dash Attack {
  path=ItemFinalPoseOverride.AnimGraphNode_StateMachine_0.DashStateMachine.AnimStateNode_1.Dash Attack
  nodes [Output Animation Pose(Result=Sequence Player 'Katana_Dash_Pose'() {sequence=/KatanaGameplay/Items/Katana/Animation/Dash/Katana_Dash_Pose})]
 }
 Dash Attack Into {
  path=ItemFinalPoseOverride.AnimGraphNode_StateMachine_0.DashStateMachine.AnimStateNode_3.Dash Attack Into
  nodes [Output Animation Pose(Result=Sequence Player 'Katana_Dash_Into'() {bLoopAnimation=false, sequence=/KatanaGameplay/Items/Katana/Animation/Dash/Katana_Dash_Into})]
 }
 Default {
  path=ItemFinalPoseOverride.AnimGraphNode_StateMachine_0.DashStateMachine.AnimStateNode_4.Default
  nodes [Output Animation Pose(Result=Use cached pose 'CachedFinalPose'())]
 }
 Dash Outro {
  path=ItemFinalPoseOverride.AnimGraphNode_StateMachine_0.DashStateMachine.AnimStateNode_5.Dash Outro
  nodes [
   n1 = Sequence Player 'Katana_Dash_Outro_Ground'() {bLoopAnimation=false, groupName=Outro, method=SyncGroup, sequence=/KatanaGameplay/Items/Katana/Animation/Dash/Katana_Dash_Outro_Ground}
   n2 = Sequence Player 'Katana_Dash_Outro'() {bLoopAnimation=false, groupName=Outro, method=SyncGroup, sequence=/KatanaGameplay/Items/Katana/Animation/Dash/Katana_Dash_Outro}
   Output Animation Pose(Result=Blend Poses by bool(BlendPose_0=@n1, BlendPose_1=@n2, bActiveValue=Get bIsOnGround() -> (bIsOnGround:bool)) {blendType=Linear})
  ]
 }
 Transition {
  path=ItemFinalPoseOverride.AnimGraphNode_StateMachine_0.DashStateMachine.AnimStateTransitionNode_0.Transition
  nodes [Result(bCanEnterTransition=Get IsDashCharge() -> (IsDashCharge:bool))]
 }
 Transition {
  path=ItemFinalPoseOverride.AnimGraphNode_StateMachine_0.DashStateMachine.AnimStateTransitionNode_1.Transition
  nodes [Result(bCanEnterTransition=Get StartedDashOutro())]
 }
 Transition {
  path=ItemFinalPoseOverride.AnimGraphNode_StateMachine_0.DashStateMachine.AnimStateTransitionNode_10.Transition
  nodes [Result(bCanEnterTransition=Get CanExitDashCharge() -> (CanExitDashCharge:bool))]
 }
 Transition {
  path=ItemFinalPoseOverride.AnimGraphNode_StateMachine_0.DashStateMachine.AnimStateTransitionNode_16.Transition
  nodes [Result(bCanEnterTransition=NOT Boolean(A=Get IsTechniqueActive() -> (IsTechniqueActive:bool)))]
 }
 Transition {
  path=ItemFinalPoseOverride.AnimGraphNode_StateMachine_0.DashStateMachine.AnimStateTransitionNode_2.Transition
  nodes [Result(bCanEnterTransition=Get IsDashAttack())]
 }
 Transition {path=ItemFinalPoseOverride.AnimGraphNode_StateMachine_0.DashStateMachine.AnimStateTransitionNode_3.Transition, nodes [Result()]}
 Transition {
  path=ItemFinalPoseOverride.AnimGraphNode_StateMachine_0.DashStateMachine.AnimStateTransitionNode_4.Transition
  nodes [Result(bCanEnterTransition=Get EarlyExitFromDashOutro() -> (EarlyExitFromDashOutro:bool))]
 }
 Transition {
  path=ItemFinalPoseOverride.AnimGraphNode_StateMachine_0.DashStateMachine.AnimStateTransitionNode_5.Transition
  nodes [Result(bCanEnterTransition=NOT Boolean(A=Get IsTechniqueActive()))]
 }
 Transition {
  path=ItemFinalPoseOverride.AnimGraphNode_StateMachine_0.DashStateMachine.AnimStateTransitionNode_6.Transition
  nodes [Result(bCanEnterTransition=Get IsDashCharge())]
 }
 Transition {
  path=ItemFinalPoseOverride.AnimGraphNode_StateMachine_0.DashStateMachine.AnimStateTransitionNode_7.Transition
  nodes [Result(bCanEnterTransition=NOT Boolean(A=Get IsTechniqueActive()))]
 }
 Transition {path=ItemFinalPoseOverride.AnimGraphNode_StateMachine_0.DashStateMachine.AnimStateTransitionNode_8.Transition, nodes [Result()]}
 Transition {
  path=ItemFinalPoseOverride.AnimGraphNode_StateMachine_0.DashStateMachine.AnimStateTransitionNode_9.Transition
  nodes [Result(bCanEnterTransition=Get TransitionRuleToDefault() -> (TransitionRuleToDefault:bool))]
 }
 ItemFullBodyAimOffsets {nodes [
   Input Pose!() -> (Pose, FullBody_AimOffsetAlpha:double, Aim_Yaw:double, Aim_Pitch:double) {name=InPoseFullBodyAimOffsets}
   Input Pose!() -> (Pose) {name=InPosePreFullBodyAimOffsets}
   Output Pose() {layerGroup=Item Override}
  ]}
 ItemFullBodyOverride {nodes [
   Input Pose() {name=InPoseFullBody}
   Output Pose() {layerGroup=Item Override}
  ]}
 ItemIdleOverride {nodes [
   Input Pose() {name=InPose_Idle}
   Output Pose() {layerGroup=Item Override}
  ]}
 ItemJetPackJumpAdditive {nodes [Output Pose() {layerGroup=Item Override}]}
 ItemJetPackStartAdditive {nodes [Output Pose() {layerGroup=Item Override}]}
 ItemJumpLoopAdditive {nodes [
   Input Pose() {name=InPoseItemJumpLoopAdditive}
   Output Pose() {layerGroup=Item Override}
  ]}
 ItemJumpOverride {nodes [
   Input Pose() {name=InPoseJump}
   Output Pose() {layerGroup=Item Override}
  ]}
 ItemJumpUpAdditive {nodes [
   Input Pose() {name=InPoseItemJumpUpAdditive}
   Output Pose() {layerGroup=Item Override}
  ]}
 ItemPreAimOffsets {nodes [
   Input Pose() {name=InPosePreAimOffsets}
   Output Pose() {layerGroup=Item Override}
  ]}
 ItemPreIK {nodes [
   Input Pose() {name=InPose_PreIK}
   Output Pose() {layerGroup=Item Override}
  ]}
 ItemRightHandIK {nodes [
   Input Pose() {name=InPosePreRightHandIK}
   Input Pose() {name=InPosePostRightHandIK}
   Output Pose() {layerGroup=Item Override}
  ]}
 ItemSkeletalControl {nodes [
   Input Pose()
   Output Pose() {layerGroup=Item Override}
  ]}
 ItemSwimJumpFallLoop {nodes [Output Pose() {layerGroup=Item Override}]}
 ItemSwimJumpSurfaceEnd {nodes [Output Pose() {layerGroup=Item Override}]}
 ItemSwimJumpSurfaceLoop {nodes [Output Pose() {layerGroup=Item Override}]}
 ItemUpperBody {nodes [
   Input Pose!() -> (Pose, InputParam:Fort Anim Input Adjusted Aim Structure) {name=InPoseUpperBody}
   Output Pose() {layerGroup=Item Override}
  ]}
 ItemZipLineStartAdditive {nodes [Output Pose() {layerGroup=Item Override}]}
]
variables [
 Character Movement {type=CharacterMovementComponent, default=null}
 DashOutroTimerHandle {type=TimerHandle}
 RightWeaponRotOffsetSlide {type=Rotator, default=<rot 180 0 -80>}
]
