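# Animation Blueprint for the katana weapon skeleton, playing sword animations for heavy spins, dash and in-air attacks.
# A WeaponAnim state machine uses a TechniqueSelector conduit and bool-driven rules to pick attack states or the reference pose.
# Its output passes through DefaultSlot, then world-space Modify Bone nodes for Scabbard (weighted by the LockScabbardOnHip curve) and handle.
# A final Inertialization node smooths the zero-crossfade transitions that use inertial blending.
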AnimBlueprint /KatanaGameplay/Items/Katana/Animation/Katana_Weapon_AnimBP
target_skeleton=/KatanaGameplay/Meshes/Katana_Skeleton
graphs [
 AnimGraph {nodes [
   AnimGraphNode_StateMachine_0 = WeaponAnim() {is=title AnimGraphNode_StateMachine, bAllowConduitEntryStates=true}
   n1 = "Transform (Modify) Bone - Bone: Scabbard"!(
    ComponentPose=Local To Component(LocalPose=Slot 'DefaultSlot'(Source=@AnimGraphNode_StateMachine_0))
    AlphaCurveName:Name=LockScabbardOnHip
    Translation="0,0,0"
    Rotation="0,0,0"
   ) -> (Pose) {
    alphaInputType=Curve
    rotationMode=BMM_Replace
    rotationSpace=BCS_WorldSpace
    translationMode=BMM_Replace
    translationSpace=BCS_WorldSpace
    alphaBoolBlend {blendOption=Linear}
    boneToModify {boneName=Scabbard}
   }
   n2 = "Transform (Modify) Bone - Bone: Melee_TwoHanded_Handle"!(ComponentPose=@n1, Alpha=1, Translation="0,0,0", Rotation="0,0,0") -> (Pose) {
    rotationMode=BMM_Replace
    rotationSpace=BCS_WorldSpace
    translationMode=BMM_Replace
    translationSpace=BCS_WorldSpace
    alphaBoolBlend {blendOption=Linear}
    boneToModify {boneName=Melee_TwoHanded_Handle}
   }
   Output Pose(Result=Inertialization!(Source:Pose Link Structure=Component To Local(ComponentPose=@n2)) -> (Pose:Pose Link Structure) {is=title AnimGraphNode_Inertialization}) {layerGroup=None}
  ]}
 WeaponAnim {
  entry=TechniqueSelector
  path=AnimGraph.AnimGraphNode_StateMachine_0.WeaponAnim
  ids=id
  nodes [
   AnimStateConduitNode_1 = TechniqueSelector() {is=title AnimStateConduitNode}
   AnimStateNode_0 = HeavySpin2() {is=title AnimStateNode}
   AnimStateNode_10 = DashCharge() {is=title AnimStateNode}
   AnimStateNode_11 = DashOutro() {
    is=title AnimStateNode
    stateEntered {guid=#1}
    stateFullyBlended {guid=#2}
    stateLeft {guid=#3}
   }
   AnimStateNode_12 = InAirIntro() {is=title AnimStateNode}
   AnimStateNode_13 = InAirLoop() {is=title AnimStateNode}
   AnimStateNode_14 = InAirLand() {is=title AnimStateNode}
   AnimStateNode_15 = DashInto() {
    is=title AnimStateNode
    stateEntered {guid=#4}
    stateFullyBlended {guid=#5}
    stateLeft {guid=#6}
   }
   AnimStateNode_16 = DashLoop() {is=title AnimStateNode}
   AnimStateNode_8 = DefaultPose() {
    is=title AnimStateNode
    stateEntered {guid=#7}
    stateFullyBlended {guid=#8}
    stateLeft {guid=#9}
   }
   AnimStateNode_9 = HeavySpin() {is=title AnimStateNode}
   AnimStateTransitionNode_0 = AnimStateTransitionNode() {crossfadeDuration=0, logicType=TLT_Inertialization, priorityOrder=4}
   AnimStateTransitionNode_1 = AnimStateTransitionNode()
   AnimStateTransitionNode_11 = AnimStateTransitionNode() {crossfadeDuration=0, priorityOrder=100}
   AnimStateTransitionNode_13 = AnimStateTransitionNode()
   AnimStateTransitionNode_17 = AnimStateTransitionNode() {crossfadeDuration=0, logicType=TLT_Inertialization, priorityOrder=3}
   AnimStateTransitionNode_18 = AnimStateTransitionNode()
   AnimStateTransitionNode_19 = AnimStateTransitionNode() {
    transitionEnd {guid=#10}
    transitionInterrupt {guid=#11}
    transitionStart {guid=#12}
   }
   AnimStateTransitionNode_2 = AnimStateTransitionNode() {crossfadeDuration=0, logicType=TLT_Inertialization, priorityOrder=2}
   AnimStateTransitionNode_22 = AnimStateTransitionNode() {crossfadeDuration=0, logicType=TLT_Inertialization, priorityOrder=0}
   AnimStateTransitionNode_24 = AnimStateTransitionNode() {crossfadeDuration=0}
   AnimStateTransitionNode_25 = AnimStateTransitionNode() {crossfadeDuration=0, logicType=TLT_Inertialization, priorityOrder=0}
   AnimStateTransitionNode_26 = AnimStateTransitionNode() {bAutomaticRuleBasedOnSequencePlayerInState=true, crossfadeDuration=0, logicType=TLT_Inertialization, priorityOrder=0}
   AnimStateTransitionNode_27 = AnimStateTransitionNode() {bAutomaticRuleBasedOnSequencePlayerInState=true, crossfadeDuration=0, priorityOrder=2}
   AnimStateTransitionNode_28 = AnimStateTransitionNode() {crossfadeDuration=0}
   AnimStateTransitionNode_29 = AnimStateTransitionNode() {crossfadeDuration=0, logicType=TLT_Inertialization}
   AnimStateTransitionNode_3 = AnimStateTransitionNode() {crossfadeDuration=0, logicType=TLT_Inertialization}
   AnimStateTransitionNode_30 = AnimStateTransitionNode() {bAutomaticRuleBasedOnSequencePlayerInState=true, crossfadeDuration=0, priorityOrder=2}
   AnimStateTransitionNode_31 = AnimStateTransitionNode() {bAutomaticRuleBasedOnSequencePlayerInState=true}
   AnimStateTransitionNode_35 = AnimStateTransitionNode() {crossfadeDuration=0}
   AnimStateTransitionNode_36 = AnimStateTransitionNode() {crossfadeDuration=0}
   AnimStateTransitionNode_37 = AnimStateTransitionNode()
   AnimStateTransitionNode_4 = AnimStateTransitionNode()
   AnimStateTransitionNode_5 = AnimStateTransitionNode() {crossfadeDuration=0, priorityOrder=2}
   AnimStateTransitionNode_6 = AnimStateTransitionNode()
   AnimStateTransitionNode_7 = AnimStateTransitionNode()
   AnimStateTransitionNode_8 = AnimStateTransitionNode() {crossfadeDuration=0, logicType=TLT_Inertialization, priorityOrder=0}
   AnimStateTransitionNode_9 = AnimStateTransitionNode()
  ]
  links [
   {
    dst=DashInto
    rule=AnimGraph.AnimGraphNode_StateMachine_0.WeaponAnim.AnimStateTransitionNode_25.Transition
    src=DashCharge
    via=AnimStateTransitionNode_25
   }
   {
    dst=DefaultPose
    rule=AnimGraph.AnimGraphNode_StateMachine_0.WeaponAnim.AnimStateTransitionNode_7.Transition
    src=DashCharge
    via=AnimStateTransitionNode_7
   }
   {
    dst=DefaultPose
    rule=AnimGraph.AnimGraphNode_StateMachine_0.WeaponAnim.AnimStateTransitionNode_9.Transition
    src=DashCharge
    via=AnimStateTransitionNode_9
   }
   {
    dst=DashLoop
    rule=AnimGraph.AnimGraphNode_StateMachine_0.WeaponAnim.AnimStateTransitionNode_26.Transition
    src=DashInto
    via=AnimStateTransitionNode_26
   }
   {
    dst=DashLoop
    rule=AnimGraph.AnimGraphNode_StateMachine_0.WeaponAnim.AnimStateTransitionNode_6.Transition
    src=DashInto
    via=AnimStateTransitionNode_6
   }
   {
    dst=DashOutro
    rule=AnimGraph.AnimGraphNode_StateMachine_0.WeaponAnim.AnimStateTransitionNode_28.Transition
    src=DashLoop
    via=AnimStateTransitionNode_28
   }
   {
    dst=DashOutro
    rule=AnimGraph.AnimGraphNode_StateMachine_0.WeaponAnim.AnimStateTransitionNode_5.Transition
    src=DashLoop
    via=AnimStateTransitionNode_5
   }
   {
    dst=DashCharge
    rule=AnimGraph.AnimGraphNode_StateMachine_0.WeaponAnim.AnimStateTransitionNode_29.Transition
    src=DashOutro
    via=AnimStateTransitionNode_29
   }
   {
    dst=DefaultPose
    rule=AnimGraph.AnimGraphNode_StateMachine_0.WeaponAnim.AnimStateTransitionNode_30.Transition
    src=DashOutro
    via=AnimStateTransitionNode_30
   }
   {
    dst=DefaultPose
    rule=AnimGraph.AnimGraphNode_StateMachine_0.WeaponAnim.AnimStateTransitionNode_36.Transition
    src=DashOutro
    via=AnimStateTransitionNode_36
   }
   {
    dst=DefaultPose
    rule=AnimGraph.AnimGraphNode_StateMachine_0.WeaponAnim.AnimStateTransitionNode_8.Transition
    src=DashOutro
    via=AnimStateTransitionNode_8
   }
   {
    dst=DashCharge
    rule=AnimGraph.AnimGraphNode_StateMachine_0.WeaponAnim.AnimStateTransitionNode_22.Transition
    src=DefaultPose
    via=AnimStateTransitionNode_22
   }
   {
    dst=TechniqueSelector
    rule=AnimGraph.AnimGraphNode_StateMachine_0.WeaponAnim.AnimStateTransitionNode_19.AnimationTransitionGraph_0
    src=DefaultPose
    via=AnimStateTransitionNode_19
   }
   {
    dst=TechniqueSelector
    rule=AnimGraph.AnimGraphNode_StateMachine_0.WeaponAnim.AnimStateTransitionNode_18.AnimationTransitionGraph_0
    src=HeavySpin
    via=AnimStateTransitionNode_18
   }
   {
    dst=TechniqueSelector
    rule=AnimGraph.AnimGraphNode_StateMachine_0.WeaponAnim.AnimStateTransitionNode_1.Transition
    src=HeavySpin2
    via=AnimStateTransitionNode_1
   }
   {
    dst=InAirLoop
    rule=AnimGraph.AnimGraphNode_StateMachine_0.WeaponAnim.AnimStateTransitionNode_27.AnimationTransitionGraph_0
    src=InAirIntro
    via=AnimStateTransitionNode_27
   }
   {
    dst=InAirLoop
    rule=AnimGraph.AnimGraphNode_StateMachine_0.WeaponAnim.AnimStateTransitionNode_3.Transition
    src=InAirIntro
    via=AnimStateTransitionNode_3
   }
   {
    dst=TechniqueSelector
    rule=AnimGraph.AnimGraphNode_StateMachine_0.WeaponAnim.AnimStateTransitionNode_13.Transition
    src=InAirIntro
    via=AnimStateTransitionNode_13
   }
   {
    dst=TechniqueSelector
    rule=AnimGraph.AnimGraphNode_StateMachine_0.WeaponAnim.AnimStateTransitionNode_31.AnimationTransitionGraph_0
    src=InAirLand
    via=AnimStateTransitionNode_31
   }
   {
    dst=TechniqueSelector
    rule=AnimGraph.AnimGraphNode_StateMachine_0.WeaponAnim.AnimStateTransitionNode_37.Transition
    src=InAirLand
    via=AnimStateTransitionNode_37
   }
   {
    dst=InAirLand
    rule=AnimGraph.AnimGraphNode_StateMachine_0.WeaponAnim.AnimStateTransitionNode_35.Transition
    src=InAirLoop
    via=AnimStateTransitionNode_35
   }
   {
    dst=TechniqueSelector
    rule=AnimGraph.AnimGraphNode_StateMachine_0.WeaponAnim.AnimStateTransitionNode_4.Transition
    src=InAirLoop
    via=AnimStateTransitionNode_4
   }
   {
    dst=DefaultPose
    rule=AnimGraph.AnimGraphNode_StateMachine_0.WeaponAnim.AnimStateTransitionNode_11.AnimationTransitionGraph_0
    src=TechniqueSelector
    via=AnimStateTransitionNode_11
   }
   {
    dst=HeavySpin
    rule=AnimGraph.AnimGraphNode_StateMachine_0.WeaponAnim.AnimStateTransitionNode_17.AnimationTransitionGraph_0
    src=TechniqueSelector
    via=AnimStateTransitionNode_17
   }
   {
    dst=HeavySpin2
    rule=AnimGraph.AnimGraphNode_StateMachine_0.WeaponAnim.AnimStateTransitionNode_0.Transition
    src=TechniqueSelector
    via=AnimStateTransitionNode_0
   }
   {
    dst=InAirIntro
    rule=AnimGraph.AnimGraphNode_StateMachine_0.WeaponAnim.AnimStateTransitionNode_24.AnimationTransitionGraph_0
    src=TechniqueSelector
    via=AnimStateTransitionNode_24
   }
   {
    dst=InAirLand
    rule=AnimGraph.AnimGraphNode_StateMachine_0.WeaponAnim.AnimStateTransitionNode_2.Transition
    src=TechniqueSelector
    via=AnimStateTransitionNode_2
   }
  ]
 }
 TechniqueSelector {path=AnimGraph.AnimGraphNode_StateMachine_0.WeaponAnim.AnimStateConduitNode_1.TechniqueSelector, nodes [Result(bCanEnterTransition=true)]}
 HeavySpin2 {
  path=AnimGraph.AnimGraphNode_StateMachine_0.WeaponAnim.AnimStateNode_0.HeavySpin2
  nodes [Output Animation Pose(Result=Sequence Player 'Katana_Combo_HeavySpin2_Katana_Sword'!(PlayRate:float=1) -> (Pose) {bLoopAnimation=false, sequence=/KatanaGameplay/Items/Katana/Animation/Attacks/Sword/Katana_Combo_HeavySpin2_Katana_Sword})]
 }
 DashCharge {
  path=AnimGraph.AnimGraphNode_StateMachine_0.WeaponAnim.AnimStateNode_10.DashCharge
  nodes [Output Animation Pose(Result=Sequence Player 'Katana_Dash_ChargeUp_Katana_Sword'!(PlayRate:float=1) -> (Pose) {bLoopAnimation=false, sequence=/KatanaGameplay/Items/Katana/Animation/Dash/Sword/Katana_Dash_ChargeUp_Katana_Sword})]
 }
 DashOutro {
  path=AnimGraph.AnimGraphNode_StateMachine_0.WeaponAnim.AnimStateNode_11.DashOutro
  nodes [Output Animation Pose(Result=Sequence Player 'Katana_Dash_Outro_Katana_Sword'() {bLoopAnimation=false, sequence=/KatanaGameplay/Items/Katana/Animation/Dash/Sword/Katana_Dash_Outro_Katana_Sword})]
 }
 InAirIntro {
  path=AnimGraph.AnimGraphNode_StateMachine_0.WeaponAnim.AnimStateNode_12.InAirIntro
  nodes [Output Animation Pose(Result=Sequence Player 'Katana_Attack_InAir_Intro_V2_WEP'!(PlayRate:float=1) -> (Pose) {bLoopAnimation=false, sequence=/KatanaGameplay/Items/Katana/Animation/Attacks/InAir/Sword/Katana_Attack_InAir_Intro_V2_WEP})]
 }
 InAirLoop {
  path=AnimGraph.AnimGraphNode_StateMachine_0.WeaponAnim.AnimStateNode_13.InAirLoop
  nodes [Output Animation Pose(Result=Sequence Player 'Katana_Attack_InAir_Loop_V2_WEP'!(PlayRate:float=1) -> (Pose) {sequence=/KatanaGameplay/Items/Katana/Animation/Attacks/InAir/Sword/Katana_Attack_InAir_Loop_V2_WEP})]
 }
 InAirLand {
  path=AnimGraph.AnimGraphNode_StateMachine_0.WeaponAnim.AnimStateNode_14.InAirLand
  nodes [Output Animation Pose(Result=Sequence Player 'Katana_Attack_InAir_Land_V2_WEP'!(PlayRate:float=1) -> (Pose) {bLoopAnimation=false, sequence=/KatanaGameplay/Items/Katana/Animation/Attacks/InAir/Sword/Katana_Attack_InAir_Land_V2_WEP})]
 }
 DashInto {
  path=AnimGraph.AnimGraphNode_StateMachine_0.WeaponAnim.AnimStateNode_15.DashInto
  nodes [Output Animation Pose(Result=Sequence Player 'Katana_Dash_Into_Katana_Sword'!(PlayRate:float=1) -> (Pose) {bLoopAnimation=false, sequence=/KatanaGameplay/Items/Katana/Animation/Dash/Sword/Katana_Dash_Into_Katana_Sword})]
 }
 DashLoop {
  path=AnimGraph.AnimGraphNode_StateMachine_0.WeaponAnim.AnimStateNode_16.DashLoop
  nodes [Output Animation Pose(Result=Sequence Player 'Katana_Dash_Pose_Katana_Sword'!(PlayRate:float=1) -> (Pose) {bLoopAnimation=false, sequence=/KatanaGameplay/Items/Katana/Animation/Dash/Sword/Katana_Dash_Pose_Katana_Sword})]
 }
 DefaultPose {
  path=AnimGraph.AnimGraphNode_StateMachine_0.WeaponAnim.AnimStateNode_8.DefaultPose
  nodes [Output Animation Pose(Result=Local Space Ref Pose())]
 }
 HeavySpin {
  path=AnimGraph.AnimGraphNode_StateMachine_0.WeaponAnim.AnimStateNode_9.HeavySpin
  nodes [
   n1 = Sequence Player 'Katana_HardSaddle_Swings_Katana_Sword'!(PlayRate:float=1) -> (Pose) {bLoopAnimation=false, sequence=/KatanaGameplay/Items/Katana/Animation/Attacks/Sword/Katana_HardSaddle_Swings_Katana_Sword}
   n2 = Sequence Player 'Katana_GrindRail_Swing_Katana_Sword'!(PlayRate:float=1) -> (Pose) {bLoopAnimation=false, sequence=/KatanaGameplay/Items/Katana/Animation/Attacks/Sword/Katana_GrindRail_Swing_Katana_Sword}
   Output Animation Pose(Result=Blend Poses by bool(
     BlendPose_0=Blend Poses by bool(BlendPose_0=@n2, BlendPose_1=@n1, BlendTime_0=0, BlendTime_1=0) {blendType=Linear}
     BlendPose_1=Sequence Player 'Katana_Combo_HeavySpin_Katana_Sword'!(PlayRate:float=1) -> (Pose) {bLoopAnimation=false, sequence=/KatanaGameplay/Items/Katana/Animation/Attacks/Sword/Katana_Combo_HeavySpin_Katana_Sword}
     BlendTime_0=0
     BlendTime_1=0
    ) {blendType=Linear})
  ]
 }
 Transition {
  path=AnimGraph.AnimGraphNode_StateMachine_0.WeaponAnim.AnimStateTransitionNode_0.Transition
  nodes [Result(bCanEnterTransition=Get IsPrimaryAttack2() -> (IsPrimaryAttack2:bool))]
 }
 Transition {
  path=AnimGraph.AnimGraphNode_StateMachine_0.WeaponAnim.AnimStateTransitionNode_1.Transition
  nodes [Result(bCanEnterTransition=Get ExitFromPrimaryAttack2() -> (ExitFromPrimaryAttack2:bool))]
 }
 AnimationTransitionGraph_0 {
  path=AnimGraph.AnimGraphNode_StateMachine_0.WeaponAnim.AnimStateTransitionNode_11.AnimationTransitionGraph_0
  nodes [Result(bCanEnterTransition=true)]
 }
 Transition {
  path=AnimGraph.AnimGraphNode_StateMachine_0.WeaponAnim.AnimStateTransitionNode_13.Transition
  nodes [Result(bCanEnterTransition=NOT Boolean(A=Get IsTechniqueActive() -> (IsTechniqueActive:bool)))]
 }
 AnimationTransitionGraph_0 {
  path=AnimGraph.AnimGraphNode_StateMachine_0.WeaponAnim.AnimStateTransitionNode_17.AnimationTransitionGraph_0
  nodes [Result(bCanEnterTransition=Get IsPrimaryAttack() -> (IsPrimaryAttack:bool))]
 }
 AnimationTransitionGraph_0 {
  path=AnimGraph.AnimGraphNode_StateMachine_0.WeaponAnim.AnimStateTransitionNode_18.AnimationTransitionGraph_0
  nodes [Result(bCanEnterTransition=Get ExitFromPrimaryAttack() -> (ExitFromPrimaryAttack:bool))]
 }
 AnimationTransitionGraph_0 {
  path=AnimGraph.AnimGraphNode_StateMachine_0.WeaponAnim.AnimStateTransitionNode_19.AnimationTransitionGraph_0
  nodes [Result(bCanEnterTransition=Get IsComboAttack() -> (IsComboAttack:bool))]
 }
 Transition {
  path=AnimGraph.AnimGraphNode_StateMachine_0.WeaponAnim.AnimStateTransitionNode_2.Transition
  nodes [Result(bCanEnterTransition=Get IsInAirAttack_Slam() -> (IsInAirAttack_Slam:bool))]
 }
 Transition {
  path=AnimGraph.AnimGraphNode_StateMachine_0.WeaponAnim.AnimStateTransitionNode_22.Transition
  nodes [Result(bCanEnterTransition=Get IsDashCharge() -> (IsDashCharge:bool))]
 }
 AnimationTransitionGraph_0 {
  path=AnimGraph.AnimGraphNode_StateMachine_0.WeaponAnim.AnimStateTransitionNode_24.AnimationTransitionGraph_0
  nodes [Result(bCanEnterTransition=Get IsInAirAttack_Leap() -> (IsInAirAttack_Leap:bool))]
 }
 Transition {
  path=AnimGraph.AnimGraphNode_StateMachine_0.WeaponAnim.AnimStateTransitionNode_25.Transition
  nodes [Result(bCanEnterTransition=Get IsDashAttack() -> (IsDashAttack:bool))]
 }
 Transition {path=AnimGraph.AnimGraphNode_StateMachine_0.WeaponAnim.AnimStateTransitionNode_26.Transition, nodes [Result()]}
 AnimationTransitionGraph_0 {path=AnimGraph.AnimGraphNode_StateMachine_0.WeaponAnim.AnimStateTransitionNode_27.AnimationTransitionGraph_0, nodes [Result()]}
 Transition {
  path=AnimGraph.AnimGraphNode_StateMachine_0.WeaponAnim.AnimStateTransitionNode_28.Transition
  nodes [Result(bCanEnterTransition=Get IsDashOutro() -> (IsDashOutro:bool))]
 }
 Transition {
  path=AnimGraph.AnimGraphNode_StateMachine_0.WeaponAnim.AnimStateTransitionNode_29.Transition
  nodes [Result(bCanEnterTransition=Get IsDashCharge())]
 }
 Transition {
  path=AnimGraph.AnimGraphNode_StateMachine_0.WeaponAnim.AnimStateTransitionNode_3.Transition
  nodes [Result(bCanEnterTransition=NOT Boolean(A=Get IsInAirAttack_Leap()))]
 }
 Transition {path=AnimGraph.AnimGraphNode_StateMachine_0.WeaponAnim.AnimStateTransitionNode_30.Transition, nodes [Result()]}
 AnimationTransitionGraph_0 {path=AnimGraph.AnimGraphNode_StateMachine_0.WeaponAnim.AnimStateTransitionNode_31.AnimationTransitionGraph_0, nodes [Result()]}
 Transition {
  path=AnimGraph.AnimGraphNode_StateMachine_0.WeaponAnim.AnimStateTransitionNode_35.Transition
  nodes [Result(bCanEnterTransition=Get IsInAirAttack_Slam())]
 }
 Transition {
  path=AnimGraph.AnimGraphNode_StateMachine_0.WeaponAnim.AnimStateTransitionNode_36.Transition
  nodes [Result(bCanEnterTransition=Get EarlyExitfromDashOutro() -> (EarlyExitfromDashOutro:bool))]
 }
 Transition {
  path=AnimGraph.AnimGraphNode_StateMachine_0.WeaponAnim.AnimStateTransitionNode_37.Transition
  nodes [Result(bCanEnterTransition=Get ExitFromInAirLand() -> (ExitFromInAirLand:bool))]
 }
 Transition {
  path=AnimGraph.AnimGraphNode_StateMachine_0.WeaponAnim.AnimStateTransitionNode_4.Transition
  nodes [Result(bCanEnterTransition=NOT Boolean(A=Get IsTechniqueActive()))]
 }
 Transition {
  path=AnimGraph.AnimGraphNode_StateMachine_0.WeaponAnim.AnimStateTransitionNode_5.Transition
  nodes [Result(bCanEnterTransition=NOT Boolean(A=Get IsTechniqueActive()))]
 }
 Transition {
  path=AnimGraph.AnimGraphNode_StateMachine_0.WeaponAnim.AnimStateTransitionNode_6.Transition
  nodes [Result(bCanEnterTransition=NOT Boolean(A=Get IsTechniqueActive()))]
 }
 Transition {
  path=AnimGraph.AnimGraphNode_StateMachine_0.WeaponAnim.AnimStateTransitionNode_7.Transition
  nodes [Result(bCanEnterTransition=NOT Boolean(A=Get IsTechniqueActive()))]
 }
 Transition {
  path=AnimGraph.AnimGraphNode_StateMachine_0.WeaponAnim.AnimStateTransitionNode_8.Transition
  nodes [Result(bCanEnterTransition=Get TransitionRuleToDefaultPose() -> (TransitionRuleToDefaultPose:bool))]
 }
 Transition {
  path=AnimGraph.AnimGraphNode_StateMachine_0.WeaponAnim.AnimStateTransitionNode_9.Transition
  nodes [Result(bCanEnterTransition=NOR Boolean!(A:bool=Get IsDashAttack(), B:bool=Get IsDashCharge()) -> (ReturnValue:bool))]
 }
 EventGraph {nodes [
   Event BlueprintUpdateAnimation()
   Event BlueprintInitializeAnimation()
  ]}
]
variables []
