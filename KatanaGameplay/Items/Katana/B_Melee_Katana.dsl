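; Katana melee weapon Blueprint that manages dash-attack charges, dash preview and cosmetic feedback.
; Local ammo acts as charges, driving an out-of-charges GameplayEffect (unless durability is used) and emissive fade or pulse timelines on a dynamic material.
; While equipped, a locally controlled timer runs dash targeting and broadcasts preview updates.
; A color variant index tints trail, impact and montage-notify VFX and picks hit decals, oriented by blade-tip movement across one tick.
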(event EventBeginPlay
  ;[0.04928 1 0.08844] Begin Play
  (|CallParentFunction)
  (bind _asfort_player_pawn (Utilities|Casting|CastToFortPlayerPawn (Actor|GetOwner (Components|GetOwner (Variables|Visual|GetWeaponMesh)))))
  (Variables|Default|SetPlayerRef _asfort_player_pawn)
  (EventDispatchers|AssignOnLocalAmmoChanged (delegate Custom|OnLocalAmmoChanged_Event))
  (bind _returnvalue (Rendering|Material|CreateDynamicMaterialInstance (Variables|Visual|GetWeaponMesh) 1 0 "None" :!class /Script/Engine.PrimitiveComponent))
  (Variables|Default|SetMID_Katana _returnvalue)
  (Components|Timeline|SetPlayRate (Variables|B_Melee_Katana|GetPulse) 2.0)
  ;[1 1 1] Setup correct Color Variant Information on FX Components
  (Effects|Components|ParticleSystem|SetIntParameter (Variables|Default|GetAnimTrail%20Niagara) "ColorVariant" (Variables|Default|GetFX%20Variant%20Index)))

(event Custom|OnLocalAmmoChanged_Event
  (params (LocalCount :type Integer) (LocalRemaining :type Integer))
  ;[0.04928 1 0.08844] Begin Play
  (CallFunction|UpdateCharges)
  (CallFunction|UpdateEmissive))

(event FortWeapon|EventOnEquippedAndAttached
  ;[0.09399 1 0.7059] Charges
  (|CallParentFunction)
  (CallFunction|UpdateCharges)
  (Utilities|FlowControl|Delay 0.2)
  (CallFunction|SetDashPreviewEnabled (ScalableFloat|GetValueasBool (Variables|Dash|Preview|GetHF_DashAttackPreviewEnabled))))

(event Custom|UpdateEmissive
  (bind _returnvalue (FortWeapon|GetLocalMagazineAmmoCount))
  (bind _mid_katana (Variables|Default|GetMID_Katana))
  ;[0.09399 1 0.7059] Emissive Control
  (if (< _returnvalue 1)
    (Variables|Default|SetbOutOfAmmo true)
    (bind (_direction _fade) (Timeline|FadeInOut ReverseFromEnd)
      (:Update
        (Rendering|Material|SetScalarParameterValue (Variables|Default|GetMID_Katana) "emissive mult" _fade :!class /Script/Engine.MaterialInstanceDynamic)))
    (else
      (Variables|Default|SetbOutOfAmmo false)
      (if (and (== _returnvalue 1) (Variables|Default|GetbOutOfAmmo))
        (Timeline|FadeInOut PlayFromStart)
        (else
          (bind (_direction_2 _pulse) (Timeline|Pulse PlayFromStart)
            (:Update
              (Rendering|Material|SetScalarParameterValue _mid_katana "emissive mult" (+ (* _pulse (Variables|Default|GetEmissivePulseFactor)) 1.0) :!class /Script/Engine.MaterialInstanceDynamic)
              (if (Utilities|Enum|Equal(Enum) _direction_2 "Forward")
                (Rendering|Material|SetScalarParameterValue _mid_katana "EmissiveSweep" _pulse :!class /Script/Engine.MaterialInstanceDynamic)))
            (:Finished
              (if (Utilities|Enum|Equal(Enum) _direction_2 "Forward")
                (Timeline|Pulse ReverseFromEnd)))))))))

(event Custom|UpdateCharges
  ;[0.09399 1 0.7059] Charges
  (if (and (not (Variables|Default|GetUseDurability?)) (< (FortWeapon|GetLocalMagazineAmmoCount) 1))
    (if (not (and (Variables|Cheats|GetbDebugIgnoreAmmo) (Editor|InEditororPIE)))
      (Creative|GameplayEffects|SimpleApplyGameplayEffectToActor "/KatanaGameplay/Items/Katana/Abilities/DashAttack/GE_Melee_Katana_Dash_OutOfCharges.GE_Melee_Katana_Dash_OutOfCharges_C" (Variables|Default|GetPlayerRef)))
    (elif (GameplayTags|HasMatchingGameplayTag (Variables|Default|GetPlayerRef) "(TagName=\"Granted.KatanaGameplay.Weapon.Melee.Katana.OutOfCharges\")")
      (GameplayEffects|RemoveActiveEffectswithGrantedTags (Ability|GetAbilitySystemComponent (Variables|Default|GetPlayerRef)) "(GameplayTags=((TagName=\"Granted.KatanaGameplay.Weapon.Melee.Katana.OutOfCharges\")))"))))

(event Custom|SetDashChargeEnabled
  (params (bIsDashChargeEnabled :type Boolean))
  ;[0.2317 0.8396 1] Dash Charge
  (bind _output_get (Variables|Dash|Charge|SetbIsChargingDash bIsDashChargeEnabled))
  (Default|CallOnDashChargeEnabledorDisabled _output_get))

(event FortWeapon|EventOnLocalUnEquip
  (|CallParentFunction)
  (CallFunction|SetDashPreviewEnabled false))

(event Custom|KatanaRefillAmmo
  ;[0 0 0] Cheat to add ammo to the Katana
  (FortWeapon|ModAmmo 3))

(event Custom|KatanaIgnoreAmmo
  (params (bIgnoreAmmo :type Boolean))
  ;[0 0 0] Cheat to not consume ammo - used by GA_Melee_Katana_DashAttack
  (Variables|Cheats|SetbDebugIgnoreAmmo bIgnoreAmmo))

(event Custom|SetDashPreviewEnabled
  (params (bIsPreviewEnabled :type Boolean))
  ;[1 0.1745 0.8218] Dash Preview
  (Networking|CanExecuteCosmeticEvents
    (:True
      (bind _playerref (Variables|Default|GetPlayerRef)
        (:then
          (if (Pawn|IsLocallyControlled _playerref)
            (bind _output_get (Variables|Dash|Preview|SetIsDashPreviewEnabled bIsPreviewEnabled))
            (Default|CallOnDashPreviewEnabledorDisabled _output_get)
            (if (Variables|Dash|Preview|GetIsDashPreviewEnabled)
              (Variables|Dash|Preview|SetDashPreviewStartTime (Utilities|Time|GetTimeSeconds))
              (bind _returnvalue (Utilities|Time|SetTimerbyEvent (EventDispatchers|CreateEvent :!event Custom|UpdateDashPreview) (ScalableFloat|GetValueatLevel (Variables|Dash|Preview|GetDashPreviewUpdateFrequency)) true))
              (Variables|Dash|Preview|SetUpdateDashPreviewTimerHandle _returnvalue)
              (CallFunction|UpdateDashPreview)
              (else
                (Utilities|Time|ClearandInvalidateTimerbyHandle (Variables|Dash|Preview|GetUpdateDashPreviewTimerHandle))))))
        (:else)))))

(event Custom|UpdateDashPreview
  ;[1 0.1745 0.8218] Dash Preview
  (bind (_outdashtargetinginfo _returnvalue) (Katana|Dash|PerformDashTargeting (Variables|Default|GetKatanaTargeting)))
  (if _returnvalue
    (Default|CallOnDashPreviewUpdated (Utilities|Struct|BreakFortKatanaDashTargetingInfo _outdashtargetinginfo) (.DashDistance (Utilities|Struct|BreakFortKatanaDashTargetingInfo _outdashtargetinginfo)) (.OutOfRangePercent (Utilities|Struct|BreakFortKatanaDashTargetingInfo _outdashtargetinginfo)) (.DashBlockingActor (Utilities|Struct|BreakFortKatanaDashTargetingInfo _outdashtargetinginfo)) (Math|Boolean|ORBoolean (Utilities|IsValid (.DashBlockingActor (Utilities|Struct|BreakFortKatanaDashTargetingInfo _outdashtargetinginfo)) :!class /Script/Engine.KismetSystemLibrary) (.bWasDashDirectionAdjusted (Utilities|Struct|BreakFortKatanaDashTargetingInfo _outdashtargetinginfo)) (> (Utilities|Array|Length (.DamagedActors (Utilities|Struct|BreakFortKatanaDashTargetingInfo _outdashtargetinginfo))) 0)))))

(event FortWeapon|EventOnPlayImpactFX
  (params (HitResult_bBlockingHit :type Boolean) (HitResult_bInitialOverlap :type Boolean) (HitResult_Time :type Float%20%28single-precision%29) (HitResult_Distance :type Float%20%28single-precision%29) (HitResult_Location :type Vector :default "0, 0, 0") (HitResult_ImpactPoint :type Vector :default "0, 0, 0") (HitResult_Normal :type Vector :default "0, 0, 0") (HitResult_ImpactNormal :type Vector :default "0, 0, 0") (HitResult_PhysMat :type Physical%20Material%20Object%20Reference) (HitResult_HitActor :type Actor%20Object%20Reference) (HitResult_HitComponent :type Primitive%20Component%20Object%20Reference) (HitResult_HitBoneName :type Name :default "None") (HitResult_BoneName :type Name :default "None") (HitResult_HitItem :type Integer) (HitResult_ElementIndex :type Integer) (HitResult_FaceIndex :type Integer) (HitResult_TraceStart :type Vector :default "0, 0, 0") (HitResult_TraceEnd :type Vector :default "0, 0, 0") (ImpactPhysicalSurface :type EPhysicalSurface%20Enum :default "SurfaceType_Default") (SpawnedPSC :type FXSystem%20Component%20Object%20Reference))
  (bind _weaponmesh (Variables|Visual|GetWeaponMesh))
  (bind _returnvalue_location (Math|Transform|BreakTransform (Transformation|GetSocketTransform _weaponmesh "FX_TrailEnd" "RTS_World")))
  (bind _returnvalue_location_2 (Math|Transform|BreakTransform (Transformation|GetSocketTransform _weaponmesh "FX_TrailEnd" "RTS_World")))
  ;[1 0.03161 0] VFX Hit Decal & Update Color Variant on Impact FXs
  (Effects|Components|ParticleSystem|SetIntParameter SpawnedPSC "ColorVariant" (Variables|Default|GetFX%20Variant%20Index))
  (bind _output_get (Variables|Default|SetHit%20Location _returnvalue_location_2))
  (Utilities|FlowControl|DelayUntilNextTick)
  (bind _output_get_2 (Variables|Default|SetNext%20Hit%20Location _returnvalue_location))
  (bind _returnvalue (Rendering|Decal|SpawnDecalatLocation (Utilities|Array|Get(acopy) (Variables|Default|GetImpactDecal) (Variables|Default|GetFX%20Variant%20Index)) "12.000000,72.000000,6.000000" HitResult_ImpactPoint (Math|Rotator|MakeRotfromXY HitResult_ImpactNormal (Math|Vector|GetUnitDirection(Vector) _output_get _output_get_2))))
  (Rendering|Components|Decal|SetFadeOut _returnvalue 0.0 1.8 false))

(event EventAnimNotify_VFX_VisualSetup
  (params (Spawn%20Notify%20VFX :type FXSystem%20Component%20Object%20Reference))
  ;[1 1 1] Set the Color Variant Info on FXs spawned from AnimNotify Events in Animation Montages
  (Effects|Components|ParticleSystem|SetIntParameter Spawn%20Notify%20VFX "ColorVariant" (Variables|Default|GetFX%20Variant%20Index)))

(fn Melee_Effect_Color
  (return "0, 0, 0"))

(fn ConstructionScript
  (|CallParentFunction))
