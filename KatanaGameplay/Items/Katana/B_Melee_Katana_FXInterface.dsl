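; Blueprint interface for katana melee weapon visual effects, named B_Melee_Katana_FXInterface.
; It declares one function, AnimNotify_VFX_VisualSetup, which takes the FX system component spawned by an animation notify.
; The function declares no outputs and has no logic here, leaving implementers to configure that spawned effect.
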(fn AnimNotify_VFX_VisualSetup
  (params (Spawn%20Notify%20VFX :type FXSystem%20Component%20Object%20Reference))
  (return))
