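# Sound cue for the katana's charge attack becoming available.
# A mixer plays two layered waves at once, with the second slightly boosted, at reduced overall volume and slightly lowered pitch.
# It borrows weapon pickup attenuation and the loot collect sound class.
# A concurrency override allows a single instance and blocks new plays while one is active.
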SoundCue /KatanaGameplay/Items/Katana/Sounds/Katana_Attack_Charge_Available_Cue
root=Mixer_0
nodes [
 {id=Mixer_0, children [WavePlayer_4,WavePlayer_5], properties {inputVolume [1,1.3]}}
 {id=WavePlayer_4, properties {soundWaveAssetPtr=&/Katana_Attack_Charge_Available_04}}
 {id=WavePlayer_5, properties {soundWaveAssetPtr=&/Katana_Attack_Charge_Available_05}}
]
properties {
 attenuationSettings=&Fort_Weapon_Pickup_Att
 bOverrideConcurrency=true
 maxDistance=3500
 pitchMultiplier=0.9
 soundClassObject=&Fort_Loot_Collect
 volumeMultiplier=0.498
 concurrencyOverrides {maxCount=1, resolutionRule=PreventNew}
}
