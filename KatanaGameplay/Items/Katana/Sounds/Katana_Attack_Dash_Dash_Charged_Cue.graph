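# Sound cue for the katana's charged dash attack, with its volume multiplier set to zero.
# A mixer layers a modulated random pick of three charged-dash waves with a random pick of three dash-swing waves delayed a quarter second.
# The mix passes through a Doppler node at the root.
# It references a katana dash attenuation asset, a melee weapon sound class and a global-limit concurrency set.
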SoundCue /KatanaGameplay/Items/Katana/Sounds/Katana_Attack_Dash_Dash_Charged_Cue
root=Doppler_0
nodes [
 {id=Delay_0, children [Random_1], properties {delayMax=0.25, delayMin=0.25}}
 {id=Doppler_0, children [Mixer_0]}
 {id=Mixer_0, children [Modulator_0,Delay_0], properties {inputVolume [1,1]}}
 {id=Modulator_0, children [Random_0]}
 {id=Random_0, children [WavePlayer_0,WavePlayer_1,WavePlayer_6], properties {weights [1,1,1]}}
 {id=Random_1, children [WavePlayer_2,WavePlayer_3,WavePlayer_4], properties {weights [1,1,1]}}
 {id=WavePlayer_0, properties {soundWaveAssetPtr=&/Katana_Attack_Dash_Dash_Charged_01}}
 {id=WavePlayer_1, properties {soundWaveAssetPtr=&/Katana_Attack_Dash_Dash_Charged_02}}
 {id=WavePlayer_2, properties {soundWaveAssetPtr=&/Katana_Dash_Swing_01}}
 {id=WavePlayer_3, properties {soundWaveAssetPtr=&/Katana_Dash_Swing_02}}
 {id=WavePlayer_4, properties {soundWaveAssetPtr=&/Katana_Dash_Swing_03}}
 {id=WavePlayer_6, properties {soundWaveAssetPtr=&/Katana_Attack_Dash_Dash_Charged_03}}
]
properties {
 attenuationSettings=&/Fort_Weapon_Melee_Katana_Dash_Linear_Att
 maxDistance=16000
 soundClassObject=&Fort_WeaponMelee_PSM
 volumeMultiplier=0
 attenuationOverrides {spatializationAlgorithm=SPATIALIZATION_HRTF}
 concurrencySet [&/Katana_Dash_GlobalLimit_Conc]
}
