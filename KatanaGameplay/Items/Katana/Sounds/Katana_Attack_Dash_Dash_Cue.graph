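# Sound cue for the katana dash attack, with linear attenuation, HRTF spatialization and a global concurrency limit.
# A quality switch feeds two Doppler branches, one with more random variations, the other with two per layer and a slightly later swing.
# Each mixes a volume-modulated DashC layer, a delayed, boosted swing and a pitch-raised DashB layer muted at zero mixer volume.
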SoundCue /KatanaGameplay/Items/Katana/Sounds/Katana_Attack_Dash_Dash_Cue
root=QualityLevel_0
nodes [
 {id=Delay_0, children [Random_1], properties {delayMax=0.2, delayMin=0.2}}
 {id=Delay_1, children [Random_3], properties {delayMax=0.25, delayMin=0.25}}
 {id=Doppler_0, children [Mixer_0]}
 {id=Doppler_1, children [Mixer_1]}
 {id=Mixer_0, children [Modulator_0,Modulator_2,Modulator_1], properties {inputVolume [1.4,0,1.5]}}
 {id=Mixer_1, children [Modulator_3,Modulator_5,Modulator_4], properties {inputVolume [1.4,0,1.5]}}
 {id=Modulator_0, children [Random_5], properties {volumeMax=1, volumeMin=0.9}}
 {id=Modulator_1, children [Delay_0]}
 {id=Modulator_2, children [Random_0], properties {pitchMax=1.2, pitchMin=1.1}}
 {id=Modulator_3, children [Random_4], properties {volumeMax=1, volumeMin=0.9}}
 {id=Modulator_4, children [Delay_1]}
 {id=Modulator_5, children [Random_2], properties {pitchMax=1.2, pitchMin=1.1}}
 {id=QualityLevel_0, children [Doppler_0,Doppler_0,Doppler_1,Doppler_1]}
 {id=Random_0, children [WavePlayer_0,WavePlayer_1,WavePlayer_6,WavePlayer_7], properties {weights [1,1,1,1]}}
 {id=Random_1, children [WavePlayer_2,WavePlayer_3,WavePlayer_4], properties {weights [1,1,1]}}
 {id=Random_2, children [WavePlayer_5,WavePlayer_8], properties {bShouldExcludeFromBranchCulling=true, weights [1,1]}}
 {id=Random_3, children [WavePlayer_11, WavePlayer_12], properties {bShouldExcludeFromBranchCulling=true, weights [1,1]}}
 {id=Random_4, children [WavePlayer_14, WavePlayer_15], properties {bShouldExcludeFromBranchCulling=true, weights [1,1]}}
 {id=Random_5, children [WavePlayer_17, WavePlayer_18, WavePlayer_19], properties {weights [1,1,1]}}
 {id=WavePlayer_0, properties {soundWaveAssetPtr=&/Katana_Attack_Dash_DashB_01}}
 {id=WavePlayer_1, properties {soundWaveAssetPtr=&/Katana_Attack_Dash_DashB_02}}
 {id=WavePlayer_11, properties {soundWaveAssetPtr=&/Katana_Dash_Swing_01}}
 {id=WavePlayer_12, properties {soundWaveAssetPtr=&/Katana_Dash_Swing_02}}
 {id=WavePlayer_14, properties {soundWaveAssetPtr=&/Katana_Attack_Dash_DashC_01}}
 {id=WavePlayer_15, properties {soundWaveAssetPtr=&/Katana_Attack_Dash_DashC_02}}
 {id=WavePlayer_17, properties {soundWaveAssetPtr=&Katana_Attack_Dash_DashC_01}}
 {id=WavePlayer_18, properties {soundWaveAssetPtr=&Katana_Attack_Dash_DashC_02}}
 {id=WavePlayer_19, properties {soundWaveAssetPtr=&/Katana_Attack_Dash_DashC_03}}
 {id=WavePlayer_2, properties {soundWaveAssetPtr=&Katana_Dash_Swing_01}}
 {id=WavePlayer_3, properties {soundWaveAssetPtr=&Katana_Dash_Swing_02}}
 {id=WavePlayer_4, properties {soundWaveAssetPtr=&/Katana_Dash_Swing_03}}
 {id=WavePlayer_5, properties {soundWaveAssetPtr=&Katana_Attack_Dash_DashB_01}}
 {id=WavePlayer_6, properties {soundWaveAssetPtr=&/Katana_Attack_Dash_DashB_03}}
 {id=WavePlayer_7, properties {soundWaveAssetPtr=&/Katana_Attack_Dash_DashB_04}}
 {id=WavePlayer_8, properties {soundWaveAssetPtr=&Katana_Attack_Dash_DashB_02}}
]
properties {
 attenuationSettings=&/Fort_Weapon_Melee_Katana_Dash_Linear_Att
 maxDistance=16000
 soundClassObject=&Fort_WeaponMelee_PSM
 volumeMultiplier=1
 attenuationOverrides {spatializationAlgorithm=SPATIALIZATION_HRTF}
 concurrencySet [&/Katana_Dash_GlobalLimit_Conc]
}
