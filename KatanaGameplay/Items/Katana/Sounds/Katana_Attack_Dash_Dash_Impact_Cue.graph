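# Sound cue for the katana dash attack impact, layering a deflect clang with a world-impact hit.
# A mixer blends two branches, each picking one of three equally weighted waves through a random node and a modulator for variation.
# The deflect layer plays immediately, while the dash world-impact layer follows after a short fixed delay.
# The cue uses katana dash attenuation, the melee weapon sound class and player melee concurrency.
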SoundCue /KatanaGameplay/Items/Katana/Sounds/Katana_Attack_Dash_Dash_Impact_Cue
root=Mixer_0
nodes [
 {id=Delay_0, children [Modulator_0], properties {delayMax=0.1, delayMin=0.1}}
 {id=Mixer_0, children [Delay_0,Modulator_1], properties {inputVolume [1,1]}}
 {id=Modulator_0, children [Random_0]}
 {id=Modulator_1, children [Random_1]}
 {id=Random_0, children [WavePlayer_0,WavePlayer_1,WavePlayer_2], properties {weights [1,1,1]}}
 {id=Random_1, children [WavePlayer_3,WavePlayer_4,WavePlayer_5], properties {weights [1,1,1]}}
 {id=WavePlayer_0, properties {soundWaveAssetPtr=&/Katana_DashImpactWorld_01}}
 {id=WavePlayer_1, properties {soundWaveAssetPtr=&/Katana_DashImpactWorld_02}}
 {id=WavePlayer_2, properties {soundWaveAssetPtr=&/Katana_DashImpactWorld_03}}
 {id=WavePlayer_3, properties {soundWaveAssetPtr=&/Katana_Block_Deflect_01}}
 {id=WavePlayer_4, properties {soundWaveAssetPtr=&/Katana_Block_Deflect_02}}
 {id=WavePlayer_5, properties {soundWaveAssetPtr=&/Katana_Block_Deflect_03}}
]
properties {
 attenuationSettings=&/Fort_Weapon_Melee_Katana_Dash_Linear_Att
 maxDistance=16000
 soundClassObject=&Fort_WeaponMelee_PSM
 volumeMultiplier=0.9
 attenuationOverrides {bAttenuate=false}
 concurrencyOverrides {resolutionRule=PreventNew}
 concurrencySet [&Player_Melee_Weapon_Conc]
}
