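# Sound cue for the katana's dash attack loop, playing a single loop-start wave in the melee weapon sound class.
# A mixer boosts that wave and leaves its other input empty and silent; no looping node is present.
# Playback uses a shared katana dash linear attenuation asset and a global-limit concurrency set.
# An HRTF spatialization value sits in the attenuation overrides, but the override is not shown as enabled.
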SoundCue /KatanaGameplay/Items/Katana/Sounds/Katana_Attack_Dash_Loop_Cue
root=Mixer_0
nodes [
 {id=Mixer_0, children [null,WavePlayer_2], properties {inputVolume [0,2]}}
 {id=WavePlayer_2, properties {soundWaveAssetPtr=&/Katana_Attack_Dash_Loop_Start_03}}
]
properties {
 attenuationSettings=&/Fort_Weapon_Melee_Katana_Dash_Linear_Att
 maxDistance=16000
 soundClassObject=&Fort_WeaponMelee_PSM
 volumeMultiplier=0.65
 attenuationOverrides {spatializationAlgorithm=SPATIALIZATION_HRTF}
 concurrencySet [&/Katana_Dash_GlobalLimit_Conc]
}
