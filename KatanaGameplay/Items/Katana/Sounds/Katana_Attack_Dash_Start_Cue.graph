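# Sound cue for the start of the Katana's dash attack.
# A Random node picks one of three equally weighted StartB wave variations.
# It routes through the melee weapon sound class with a volume boost, uses a linear dash attenuation asset, and sets HRTF spatialization in its attenuation overrides.
# A global-limit concurrency asset governs simultaneous plays.
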SoundCue /KatanaGameplay/Items/Katana/Sounds/Katana_Attack_Dash_Start_Cue
root=Random_1
nodes [
 {id=Random_1, children [WavePlayer_0,WavePlayer_1,WavePlayer_2], properties {weights [1,1,1]}}
 {id=WavePlayer_0, properties {soundWaveAssetPtr=&/Katana_Attack_Dash_StartB_01}}
 {id=WavePlayer_1, properties {soundWaveAssetPtr=&/Katana_Attack_Dash_StartB_02}}
 {id=WavePlayer_2, properties {soundWaveAssetPtr=&/Katana_Attack_Dash_StartB_03}}
]
properties {
 attenuationSettings=&/Fort_Weapon_Melee_Katana_Dash_Linear_Att
 maxDistance=16000
 soundClassObject=&Fort_WeaponMelee_PSM
 volumeMultiplier=1.3
 attenuationOverrides {spatializationAlgorithm=SPATIALIZATION_HRTF}
 concurrencySet [&/Katana_Dash_GlobalLimit_Conc]
}
