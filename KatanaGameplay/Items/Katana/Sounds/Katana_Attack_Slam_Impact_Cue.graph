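# Sound cue for the impact of the katana slam attack, layering ground hit, boom and swing sounds.
# A mixer blends three branches, with the ground layer loudest and the boom quietest, each picking randomly from its wave variants.
# Ground and boom pitch varies upward, while the reused second-swing layer is pitched slightly down.
# It uses close-range melee attenuation, HRTF spatialization as an override and the katana dash global concurrency limit.
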SoundCue /KatanaGameplay/Items/Katana/Sounds/Katana_Attack_Slam_Impact_Cue
root=Mixer_4
nodes [
 {id=Mixer_4, children [Modulator_1,Modulator_0,Modulator_2], properties {inputVolume [1.3,0.5,1]}}
 {id=Modulator_0, children [Random_8], properties {pitchMax=1.3, pitchMin=1}}
 {id=Modulator_1, children [Random_7], properties {pitchMax=1.3, pitchMin=1}}
 {id=Modulator_2, children [Random_1], properties {pitchMax=0.9, pitchMin=0.9}}
 {id=Random_1, children [WavePlayer_18, WavePlayer_21, WavePlayer_22, WavePlayer_23], properties {weights [1,1,1,1]}}
 {id=Random_7, children [WavePlayer_6, WavePlayer_7, WavePlayer_24], properties {weights [1,1,1]}}
 {id=Random_8, children [WavePlayer_25, WavePlayer_26, WavePlayer_27], properties {weights [1,1,1]}}
 {id=WavePlayer_18, properties {soundWaveAssetPtr=&/Katana_Attack_Swing_Second_A_01}}
 {id=WavePlayer_21, properties {soundWaveAssetPtr=&/Katana_Attack_Swing_Second_A_02}}
 {id=WavePlayer_22, properties {soundWaveAssetPtr=&/Katana_Attack_Swing_Second_A_03}}
 {id=WavePlayer_23, properties {soundWaveAssetPtr=&/Katana_Attack_Swing_Second_A_04}}
 {id=WavePlayer_24, properties {soundWaveAssetPtr=&/Katana_Attack_Slam_Impact_Ground_03}}
 {id=WavePlayer_25, properties {soundWaveAssetPtr=&/Katana_Attack_Slam_Impact_Boom_01}}
 {id=WavePlayer_26, properties {soundWaveAssetPtr=&/Katana_Attack_Slam_Impact_Boom_02}}
 {id=WavePlayer_27, properties {soundWaveAssetPtr=&/Katana_Attack_Slam_Impact_Boom_03}}
 {id=WavePlayer_6, properties {soundWaveAssetPtr=&/Katana_Attack_Slam_Impact_Ground_01}}
 {id=WavePlayer_7, properties {soundWaveAssetPtr=&/Katana_Attack_Slam_Impact_Ground_02}}
]
properties {
 attenuationSettings=/Game/Sounds/Attenuation/Fort_Weapon_Melee_Attack_Close_Range_Stereo_Att
 maxDistance=10058.75
 soundClassObject=&Fort_WeaponMelee_PSM
 volumeMultiplier=1
 attenuationOverrides {spatializationAlgorithm=SPATIALIZATION_HRTF}
 concurrencySet [&/Katana_Dash_GlobalLimit_Conc]
}
