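# Sound cue for the start of the katana's slam attack, built from the katana dash swing sounds.
# An equally weighted Random node picks one of three swing waves, which a Modulator varies in pitch and volume before a Doppler node.
# It uses the close-range katana melee attenuation with HRTF spatialization, a melee weapon sound class and a concurrency limit named for a global katana dash cap.
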SoundCue /KatanaGameplay/Items/Katana/Sounds/Katana_Attack_Slam_Start_Cue
root=Doppler_0
nodes [
 {id=Doppler_0, children [Modulator_1]}
 {id=Modulator_1, children [Random_1], properties {pitchMax=1.1, volumeMax=1.3, volumeMin=1.1}}
 {id=Random_1, children [WavePlayer_2,WavePlayer_3,WavePlayer_4], properties {weights [1,1,1]}}
 {id=WavePlayer_2, properties {soundWaveAssetPtr=&/Katana_Dash_Swing_01}}
 {id=WavePlayer_3, properties {soundWaveAssetPtr=&/Katana_Dash_Swing_02}}
 {id=WavePlayer_4, properties {soundWaveAssetPtr=&/Katana_Dash_Swing_03}}
]
properties {
 attenuationSettings=&/Fort_Weapon_Melee_Attack_Close_Katana_Att
 maxDistance=7200
 soundClassObject=&Fort_WeaponMelee_PSM
 volumeMultiplier=0.6
 attenuationOverrides {spatializationAlgorithm=SPATIALIZATION_HRTF}
 concurrencySet [&/Katana_Dash_GlobalLimit_Conc]
}
