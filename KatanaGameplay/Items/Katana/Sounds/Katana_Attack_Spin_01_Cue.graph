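# Sound cue for the katana spin attack whoosh, routed to a melee weapon sound class with close-katana attenuation.
# Each variant mixes three randomized layers: a 0.1-second delayed first swing, a quieter swing layer B and a power swing.
# A quality switch sends the first two levels to the larger sample pools and the last two to two-sample pools.
# Modulators vary pitch, and a concurrency override caps voices per owner, stopping the oldest.
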SoundCue /KatanaGameplay/Items/Katana/Sounds/Katana_Attack_Spin_01_Cue
root=Modulator_0
nodes [
 {id=Delay_1, children [Random_6], properties {delayMax=0.1, delayMin=0.1}}
 {id=Delay_2, children [Random_2], properties {delayMax=0.1, delayMin=0.1}}
 {id=Mixer_0, children [Modulator_5,Modulator_7,Modulator_12], properties {inputVolume [1,0.5,0.9]}}
 {id=Mixer_2, children [Modulator_1,Modulator_8,Modulator_3], properties {inputVolume [1,0.4,0.9]}}
 {id=Modulator_0, children [QualityLevel_0], properties {pitchMax=1.1, pitchMin=1, volumeMin=1}}
 {id=Modulator_1, children [Delay_1]}
 {id=Modulator_12, children [Random_4]}
 {id=Modulator_3, children [Random_0]}
 {id=Modulator_4, children [Mixer_0]}
 {id=Modulator_5, children [Delay_2]}
 {id=Modulator_6, children [Mixer_2]}
 {id=Modulator_7, children [Random_3], properties {pitchMax=1, pitchMin=0.9}}
 {id=Modulator_8, children [Random_7], properties {pitchMax=1, pitchMin=0.9}}
 {id=QualityLevel_0, children [Modulator_6,Modulator_6,Modulator_4,Modulator_4]}
 {id=Random_0, children [WavePlayer_0,WavePlayer_1,WavePlayer_2,WavePlayer_6,WavePlayer_7], properties {weights [1,1,1,0.5,1]}}
 {id=Random_2, children [WavePlayer_3,WavePlayer_9], properties {bShouldExcludeFromBranchCulling=true, weights [1,1]}}
 {id=Random_3, children [WavePlayer_11, WavePlayer_12], properties {bShouldExcludeFromBranchCulling=true, weights [1,1]}}
 {id=Random_4, children [WavePlayer_16, WavePlayer_18], properties {bShouldExcludeFromBranchCulling=true, weights [1,1]}}
 {id=Random_6, children [WavePlayer_4, WavePlayer_14, WavePlayer_17], properties {weights [1,1,1]}}
 {id=Random_7, children [WavePlayer_21, WavePlayer_22, WavePlayer_23, WavePlayer_24], properties {weights [1,1,1,1]}}
 {id=WavePlayer_0, properties {soundWaveAssetPtr=&/Katana_Swing_Power_A_01}}
 {id=WavePlayer_1, properties {soundWaveAssetPtr=&/Katana_Swing_Power_A_02}}
 {id=WavePlayer_11, properties {soundWaveAssetPtr=&/Katana_Attack_Swing_First_Layer_B_01}}
 {id=WavePlayer_12, properties {soundWaveAssetPtr=&/Katana_Attack_Swing_First_Layer_B_02}}
 {id=WavePlayer_14, properties {soundWaveAssetPtr=&/Katana_Attack_Swing_First_02}}
 {id=WavePlayer_16, properties {soundWaveAssetPtr=&Katana_Swing_Power_A_01}}
 {id=WavePlayer_17, properties {soundWaveAssetPtr=&/Katana_Attack_Swing_First_03}}
 {id=WavePlayer_18, properties {soundWaveAssetPtr=&Katana_Swing_Power_A_02}}
 {id=WavePlayer_2, properties {soundWaveAssetPtr=&/Katana_Swing_Power_A_03}}
 {id=WavePlayer_21, properties {soundWaveAssetPtr=&Katana_Attack_Swing_First_Layer_B_01}}
 {id=WavePlayer_22, properties {soundWaveAssetPtr=&Katana_Attack_Swing_First_Layer_B_02}}
 {id=WavePlayer_23, properties {soundWaveAssetPtr=&/Katana_Attack_Swing_First_Layer_B_03}}
 {id=WavePlayer_24, properties {soundWaveAssetPtr=&/Katana_Attack_Swing_First_Layer_B_04}}
 {id=WavePlayer_3, properties {soundWaveAssetPtr=&/Katana_Attack_Swing_First_01}}
 {id=WavePlayer_4, properties {soundWaveAssetPtr=&Katana_Attack_Swing_First_01}}
 {id=WavePlayer_6, properties {soundWaveAssetPtr=&/Katana_Swing_Power_C_02}}
 {id=WavePlayer_7, properties {soundWaveAssetPtr=&/Katana_Swing_Power_C_03}}
 {id=WavePlayer_9, properties {soundWaveAssetPtr=&Katana_Attack_Swing_First_02}}
]
properties {
 attenuationSettings=&/Fort_Weapon_Melee_Attack_Close_Katana_Att
 bOverrideConcurrency=true
 maxDistance=7200
 soundClassObject=&Fort_WeaponMelee_PSM
 volumeMultiplier=0.8
 attenuationOverrides {
  binauralRadius=500
  falloffDistance=2400
  nonSpatializedRadiusStart=500
  spatializationAlgorithm=SPATIALIZATION_HRTF
  stereoSpread=2000
  attenuationShapeExtents {x=800}
 }
 concurrencyOverrides {bLimitToOwner=true, maxCount=4, resolutionRule=StopOldest}
 concurrencySet [&/Katana_Dash_GlobalLimit_Conc]
}
