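# Sound cue for the second katana spin attack, played as a close melee weapon sound under a katana dash concurrency limit.
# A quality node picks one of two four-layer mixes: full random variation for its first two slots, a reduced wave pool for the last two.
# Each mix blends swing, secondary swing and power swing layers with a delayed stow sound.
# Nested modulators vary pitch on selected layers and the whole cue.
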SoundCue /KatanaGameplay/Items/Katana/Sounds/Katana_Attack_Spin_02_Cue
root=Modulator_8
nodes [
 {id=Delay_0, children [Random_1], properties {delayMax=0.7, delayMin=0.7}}
 {id=Delay_1, children [Random_9], properties {delayMax=0.7, delayMin=0.7}}
 {id=Mixer_0, children [Modulator_7,Modulator_5,Modulator_13,Modulator_12], properties {inputVolume [1,0.5,1.1,0.2]}}
 {id=Mixer_2, children [Modulator_1,Modulator_0,Modulator_3,Modulator_2], properties {inputVolume [1,0.4,1.1,0.2]}}
 {id=Modulator_0, children [Random_0]}
 {id=Modulator_1, children [Random_7], properties {pitchMax=1.1, pitchMin=1}}
 {id=Modulator_12, children [Delay_1], properties {pitchMax=1.1, pitchMin=1}}
 {id=Modulator_13, children [Random_10]}
 {id=Modulator_2, children [Delay_0], properties {pitchMax=1.1, pitchMin=1}}
 {id=Modulator_3, children [Random_5]}
 {id=Modulator_4, children [Mixer_0]}
 {id=Modulator_5, children [Random_8], properties {pitchMax=1, pitchMin=0.9}}
 {id=Modulator_6, children [Mixer_2], properties {pitchMax=1.1, pitchMin=1}}
 {id=Modulator_7, children [Random_6], properties {pitchMax=1.1, pitchMin=1}}
 {id=Modulator_8, children [QualityLevel_0], properties {pitchMax=1, pitchMin=0.9, volumeMin=1}}
 {id=QualityLevel_0, children [Modulator_6,Modulator_6,Modulator_4,Modulator_4]}
 {id=Random_0, children [WavePlayer_0,WavePlayer_1,WavePlayer_2,WavePlayer_3], properties {weights [1,1,1,1]}}
 {id=Random_1, children [WavePlayer_4,WavePlayer_5,WavePlayer_6], properties {weights [1,1,1]}}
 {id=Random_10, children [WavePlayer_31], properties {bShouldExcludeFromBranchCulling=true, weights [1]}}
 {
  id=Random_5
  children [WavePlayer_9, WavePlayer_10, WavePlayer_13, WavePlayer_7, WavePlayer_8, WavePlayer_11, WavePlayer_12, WavePlayer_15]
  properties {weights [1,1,1,1,1,1,1,1]}
 }
 {id=Random_6, children [WavePlayer_14, WavePlayer_17], properties {bShouldExcludeFromBranchCulling=true, weights [1,1]}}
 {id=Random_7, children [WavePlayer_18, WavePlayer_21, WavePlayer_22, WavePlayer_23], properties {weights [1,1,1,1]}}
 {id=Random_8, children [WavePlayer_24, WavePlayer_25], properties {bShouldExcludeFromBranchCulling=true, weights [1,1]}}
 {id=Random_9, children [WavePlayer_28], properties {bShouldExcludeFromBranchCulling=true, weights [1]}}
 {id=WavePlayer_0, properties {soundWaveAssetPtr=&/Katana_Attack_Swing_Second_Layer_01}}
 {id=WavePlayer_1, properties {soundWaveAssetPtr=&/Katana_Attack_Swing_Second_Layer_02}}
 {id=WavePlayer_10, properties {soundWaveAssetPtr=&/Katana_Swing_Power_B_02}}
 {id=WavePlayer_11, properties {soundWaveAssetPtr=&/Katana_Swing_Power_D_03}}
 {id=WavePlayer_12, properties {soundWaveAssetPtr=&/Katana_Swing_Power_D_04}}
 {id=WavePlayer_13, properties {soundWaveAssetPtr=&/Katana_Swing_Power_B_03}}
 {id=WavePlayer_14, properties {soundWaveAssetPtr=&/Katana_Attack_Swing_Second_A_01}}
 {id=WavePlayer_15, properties {soundWaveAssetPtr=&/Katana_Swing_Power_D_05}}
 {id=WavePlayer_17, properties {soundWaveAssetPtr=&/Katana_Attack_Swing_Second_A_02}}
 {id=WavePlayer_18, properties {soundWaveAssetPtr=&Katana_Attack_Swing_Second_A_01}}
 {id=WavePlayer_2, properties {soundWaveAssetPtr=&/Katana_Attack_Swing_Second_Layer_03}}
 {id=WavePlayer_21, properties {soundWaveAssetPtr=&Katana_Attack_Swing_Second_A_02}}
 {id=WavePlayer_22, properties {soundWaveAssetPtr=&/Katana_Attack_Swing_Second_A_03}}
 {id=WavePlayer_23, properties {soundWaveAssetPtr=&/Katana_Attack_Swing_Second_A_04}}
 {id=WavePlayer_24, properties {soundWaveAssetPtr=&Katana_Attack_Swing_Second_Layer_01}}
 {id=WavePlayer_25, properties {soundWaveAssetPtr=&Katana_Attack_Swing_Second_Layer_02}}
 {id=WavePlayer_28, properties {soundWaveAssetPtr=&/Katana_StowB_01}}
 {id=WavePlayer_3, properties {soundWaveAssetPtr=&/Katana_Attack_Swing_Second_Layer_04}}
 {id=WavePlayer_31, properties {soundWaveAssetPtr=&/Katana_Swing_Power_B_01}}
 {id=WavePlayer_4, properties {soundWaveAssetPtr=&Katana_StowB_01}}
 {id=WavePlayer_5, properties {soundWaveAssetPtr=&/Katana_StowB_02}}
 {id=WavePlayer_6, properties {soundWaveAssetPtr=&/Katana_StowB_03}}
 {id=WavePlayer_7, properties {soundWaveAssetPtr=&/Katana_Swing_Power_D_01}}
 {id=WavePlayer_8, properties {soundWaveAssetPtr=&/Katana_Swing_Power_D_02}}
 {id=WavePlayer_9, properties {soundWaveAssetPtr=&Katana_Swing_Power_B_01}}
]
properties {
 attenuationSettings=&/Fort_Weapon_Melee_Attack_Close_Katana_Att
 maxDistance=7200
 soundClassObject=&Fort_WeaponMelee_PSM
 volumeMultiplier=0.9
 attenuationOverrides {
  binauralRadius=500
  falloffDistance=2400
  nonSpatializedRadiusStart=500
  spatializationAlgorithm=SPATIALIZATION_HRTF
  stereoSpread=2000
  attenuationShapeExtents {x=800}
 }
 concurrencyOverrides {bLimitToOwner=true, maxCount=1, resolutionRule=PreventNew}
 concurrencySet [&/Katana_Dash_GlobalLimit_Conc]
}
