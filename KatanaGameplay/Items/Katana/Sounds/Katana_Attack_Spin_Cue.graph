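# Sound cue for the katana spin attack, built from Swing_Second wave variations.
# A quality-level node sends its first two levels through a modulator into a random pick of four waves, and its last two into an unmodulated random pick of two.
# It plays in the melee weapon sound class with a close-attack katana attenuation asset and shares a katana dash global-limit concurrency asset.
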SoundCue /KatanaGameplay/Items/Katana/Sounds/Katana_Attack_Spin_Cue
root=QualityLevel_0
nodes [
 {id=Modulator_6, children [Random_7]}
 {id=QualityLevel_0, children [Modulator_6,Modulator_6,Random_0,Random_0]}
 {id=Random_0, children [WavePlayer_0,WavePlayer_1], properties {bShouldExcludeFromBranchCulling=true, weights [1,1]}}
 {id=Random_7, children [WavePlayer_18, WavePlayer_21, WavePlayer_22, WavePlayer_23], properties {weights [1,1,1,1]}}
 {id=WavePlayer_0, properties {soundWaveAssetPtr=&/Katana_Attack_Swing_Second_A_01}}
 {id=WavePlayer_1, properties {soundWaveAssetPtr=&/Katana_Attack_Swing_Second_A_02}}
 {id=WavePlayer_18, properties {soundWaveAssetPtr=&Katana_Attack_Swing_Second_A_01}}
 {id=WavePlayer_21, properties {soundWaveAssetPtr=&Katana_Attack_Swing_Second_A_02}}
 {id=WavePlayer_22, properties {soundWaveAssetPtr=&/Katana_Attack_Swing_Second_A_03}}
 {id=WavePlayer_23, properties {soundWaveAssetPtr=&/Katana_Attack_Swing_Second_A_04}}
]
properties {
 attenuationSettings=&/Fort_Weapon_Melee_Attack_Close_Katana_Att
 maxDistance=7200
 soundClassObject=&Fort_WeaponMelee_PSM
 volumeMultiplier=0.8
 attenuationOverrides {
  binauralRadius=500
  falloffDistance=2400
  nonSpatializedRadiusStart=500
  spatializationAlgorithm=SPATIALIZATION_HRTF
  stereoSpread=2000
  attenuationShapeExtents {x=800}
 }
 concurrencyOverrides {bLimitToOwner=true, maxCount=1, resolutionRule=PreventNew}
 concurrencySet [&/Katana_Dash_GlobalLimit_Conc]
}
