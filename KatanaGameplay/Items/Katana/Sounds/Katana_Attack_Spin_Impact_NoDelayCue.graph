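# Sound cue for the katana spin attack impact, named as the no-delay version.
# It picks one of three equally weighted katana dash world-impact waves at random and passes it through a modulator at default settings, with a slightly raised cue pitch.
# It routes to the melee weapon sound class and references small-melee attenuation and player melee weapon concurrency assets.
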SoundCue /KatanaGameplay/Items/Katana/Sounds/Katana_Attack_Spin_Impact_NoDelayCue
root=Modulator_0
nodes [
 {id=Modulator_0, children [Random_0]}
 {id=Random_0, children [WavePlayer_1,WavePlayer_2,WavePlayer_0], properties {weights [1,1,1]}}
 {id=WavePlayer_0, properties {soundWaveAssetPtr=&/Katana_DashImpactWorld_01}}
 {id=WavePlayer_1, properties {soundWaveAssetPtr=&/Katana_DashImpactWorld_02}}
 {id=WavePlayer_2, properties {soundWaveAssetPtr=&/Katana_DashImpactWorld_03}}
]
properties {
 attenuationSettings=&Fort_Weapon_Melee_Sml_Att
 maxDistance=2800
 pitchMultiplier=1.1
 soundClassObject=&Fort_WeaponMelee_PSM
 volumeMultiplier=1
 attenuationOverrides {bAttenuate=false}
 concurrencyOverrides {resolutionRule=PreventNew}
 concurrencySet [&Player_Melee_Weapon_Conc]
}
