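# Sound cue for a katana attack swing (Fauna variant), using katana close-range stereo attenuation and player melee concurrency.
# Each branch mixes an immediate random power swing and pitched-up attack swing with delayed, pitched-down second-swing and power layers.
# A quality node sends its first two levels to four-variation pools with staggered delays.
# The last two levels use two-variation pools, excluded from branch culling, sharing one delay.
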SoundCue /KatanaGameplay/Items/Katana/Sounds/Katana_Attack_Swing_Fauna_Cue
root=QualityLevel_0
nodes [
 {id=Delay_0, children [Random_2], properties {delayMax=0.37, delayMin=0.37}}
 {id=Delay_1, children [Random_1], properties {delayMax=0.28, delayMin=0.28}}
 {id=Delay_2, children [Mixer_8], properties {delayMax=0.4, delayMin=0.4}}
 {id=Mixer_1, children [Mixer_4,Modulator_1], properties {inputVolume [0.7,1]}}
 {id=Mixer_4, children [Modulator_0,Random_0], properties {inputVolume [0.9,0.7]}}
 {id=Mixer_5, children [Delay_0,Delay_1], properties {inputVolume [1,0.7]}}
 {id=Mixer_6, children [Mixer_7,Modulator_7], properties {inputVolume [0.7,1]}}
 {id=Mixer_7, children [Modulator_6,Random_8], properties {inputVolume [0.9,0.7]}}
 {id=Mixer_8, children [Random_6,Random_9], properties {inputVolume [1,0.8]}}
 {id=Modulator_0, children [Random_3], properties {pitchMax=1.2, pitchMin=1.1}}
 {id=Modulator_1, children [Mixer_5], properties {pitchMax=0.95, pitchMin=0.85}}
 {id=Modulator_14, children [Mixer_1]}
 {id=Modulator_5, children [Mixer_6]}
 {id=Modulator_6, children [Random_7], properties {pitchMax=1.2, pitchMin=1.1}}
 {id=Modulator_7, children [Delay_2], properties {pitchMax=0.95, pitchMin=0.85}}
 {id=QualityLevel_0, children [Modulator_14,Modulator_14,Modulator_5,Modulator_5]}
 {id=Random_0, children [WavePlayer_0,WavePlayer_1,WavePlayer_2,WavePlayer_3], properties {weights [1,1,1,1]}}
 {id=Random_1, children [WavePlayer_18, WavePlayer_19, WavePlayer_22, WavePlayer_23], properties {weights [1,1,1,1]}}
 {id=Random_2, children [WavePlayer_8, WavePlayer_9, WavePlayer_10, WavePlayer_11], properties {weights [1,1,1,1]}}
 {id=Random_3, children [WavePlayer_12, WavePlayer_13, WavePlayer_14, WavePlayer_15], properties {weights [1,1,1,1]}}
 {id=Random_6, children [WavePlayer_24, WavePlayer_25], properties {bShouldExcludeFromBranchCulling=true, weights [1,1]}}
 {id=Random_7, children [WavePlayer_28, WavePlayer_29], properties {bShouldExcludeFromBranchCulling=true, weights [1,1]}}
 {id=Random_8, children [WavePlayer_32, WavePlayer_33], properties {bShouldExcludeFromBranchCulling=true, weights [1,1]}}
 {id=Random_9, children [WavePlayer_36, WavePlayer_37], properties {bShouldExcludeFromBranchCulling=true, weights [1,1]}}
 {id=WavePlayer_0, properties {soundWaveAssetPtr=&/Katana_Swing_Power_C_03}}
 {id=WavePlayer_1, properties {soundWaveAssetPtr=&/Katana_Swing_Power_C_02}}
 {id=WavePlayer_10, properties {soundWaveAssetPtr=&/Katana_Attack_Swing_Second_A_03}}
 {id=WavePlayer_11, properties {soundWaveAssetPtr=&/Katana_Attack_Swing_Second_A_04}}
 {id=WavePlayer_12, properties {soundWaveAssetPtr=&/Katana_Attack_Swing_01}}
 {id=WavePlayer_13, properties {soundWaveAssetPtr=&/Katana_Attack_Swing_02}}
 {id=WavePlayer_14, properties {soundWaveAssetPtr=&/Katana_Attack_Swing_03}}
 {id=WavePlayer_15, properties {soundWaveAssetPtr=&/Katana_Attack_Swing_04}}
 {id=WavePlayer_18, properties {soundWaveAssetPtr=&/Katana_Swing_Power_D_01}}
 {id=WavePlayer_19, properties {soundWaveAssetPtr=&/Katana_Swing_Power_D_02}}
 {id=WavePlayer_2, properties {soundWaveAssetPtr=&/Katana_Swing_Power_A_01}}
 {id=WavePlayer_22, properties {soundWaveAssetPtr=&/Katana_Swing_Power_D_03}}
 {id=WavePlayer_23, properties {soundWaveAssetPtr=&/Katana_Swing_Power_D_04}}
 {id=WavePlayer_24, properties {soundWaveAssetPtr=&/Katana_Attack_Swing_Second_A_01}}
 {id=WavePlayer_25, properties {soundWaveAssetPtr=&/Katana_Attack_Swing_Second_A_02}}
 {id=WavePlayer_28, properties {soundWaveAssetPtr=&Katana_Attack_Swing_01}}
 {id=WavePlayer_29, properties {soundWaveAssetPtr=&Katana_Attack_Swing_02}}
 {id=WavePlayer_3, properties {soundWaveAssetPtr=&/Katana_Swing_Power_A_02}}
 {id=WavePlayer_32, properties {soundWaveAssetPtr=&Katana_Swing_Power_C_03}}
 {id=WavePlayer_33, properties {soundWaveAssetPtr=&Katana_Swing_Power_C_02}}
 {id=WavePlayer_36, properties {soundWaveAssetPtr=&Katana_Swing_Power_D_01}}
 {id=WavePlayer_37, properties {soundWaveAssetPtr=&Katana_Swing_Power_D_02}}
 {id=WavePlayer_8, properties {soundWaveAssetPtr=&Katana_Attack_Swing_Second_A_01}}
 {id=WavePlayer_9, properties {soundWaveAssetPtr=&Katana_Attack_Swing_Second_A_02}}
]
properties {
 attenuationSettings=&/Fort_Weapon_Melee_Katana_Close_Range_Stereo_Att
 maxDistance=6000
 soundClassObject=&Fort_WeaponMelee_PSM
 volumeMultiplier=0.9
 concurrencySet [&Player_Melee_Weapon_Conc]
}
