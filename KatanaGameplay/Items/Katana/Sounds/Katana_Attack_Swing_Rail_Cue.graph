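# Sound cue for the katana's rail attack swing, played with close-range melee attenuation and player melee weapon concurrency.
# Each play layers a pitched-up random swing and a random power swing, followed 0.4 seconds later by a pitched-down pair of second-swing and power variants.
# A quality level node feeds the first two levels four variants per layer and the last two a duplicate branch trimmed to two, whose random nodes are excluded from branch culling.
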SoundCue /KatanaGameplay/Items/Katana/Sounds/Katana_Attack_Swing_Rail_Cue
root=QualityLevel_1
nodes [
 {id=Delay_2, children [Mixer_5], properties {delayMax=0.4, delayMin=0.4}}
 {id=Delay_3, children [Mixer_8], properties {delayMax=0.4, delayMin=0.4}}
 {id=Mixer_0, children [Mixer_4,Modulator_7], properties {inputVolume [0.7,1]}}
 {id=Mixer_4, children [Modulator_6,Random_6], properties {inputVolume [0.9,0.7]}}
 {id=Mixer_5, children [Random_0,Random_7], properties {inputVolume [1,0.8]}}
 {id=Mixer_6, children [Mixer_7,Modulator_10], properties {inputVolume [0.7,1]}}
 {id=Mixer_7, children [Modulator_9,Random_10], properties {inputVolume [0.9,0.7]}}
 {id=Mixer_8, children [Random_8,Random_11], properties {inputVolume [1,0.8]}}
 {id=Modulator_0, children [Mixer_0]}
 {id=Modulator_10, children [Delay_3], properties {pitchMax=0.95, pitchMin=0.85}}
 {id=Modulator_6, children [Random_1], properties {pitchMax=1.2, pitchMin=1.1}}
 {id=Modulator_7, children [Delay_2], properties {pitchMax=0.95, pitchMin=0.85}}
 {id=Modulator_8, children [Mixer_6]}
 {id=Modulator_9, children [Random_9], properties {pitchMax=1.2, pitchMin=1.1}}
 {id=QualityLevel_1, children [Modulator_0,Modulator_0,Modulator_8,Modulator_8]}
 {id=Random_0, children [WavePlayer_0,WavePlayer_1,WavePlayer_2,WavePlayer_3], properties {weights [1,1,1,1]}}
 {id=Random_1, children [WavePlayer_4,WavePlayer_5,WavePlayer_6,WavePlayer_7], properties {weights [1,1,1,1]}}
 {id=Random_10, children [WavePlayer_32, WavePlayer_33], properties {bShouldExcludeFromBranchCulling=true, weights [1,1]}}
 {id=Random_11, children [WavePlayer_36, WavePlayer_37], properties {bShouldExcludeFromBranchCulling=true, weights [1,1]}}
 {id=Random_6, children [WavePlayer_18, WavePlayer_19, WavePlayer_22, WavePlayer_23], properties {weights [1,1,1,1]}}
 {id=Random_7, children [WavePlayer_24, WavePlayer_25, WavePlayer_26, WavePlayer_27], properties {weights [1,1,1,1]}}
 {id=Random_8, children [WavePlayer_28, WavePlayer_29], properties {bShouldExcludeFromBranchCulling=true, weights [1,1]}}
 {id=Random_9, children [WavePlayer_30, WavePlayer_31], properties {bShouldExcludeFromBranchCulling=true, weights [1,1]}}
 {id=WavePlayer_0, properties {soundWaveAssetPtr=&/Katana_Attack_Swing_Second_A_01}}
 {id=WavePlayer_1, properties {soundWaveAssetPtr=&/Katana_Attack_Swing_Second_A_02}}
 {id=WavePlayer_18, properties {soundWaveAssetPtr=&/Katana_Swing_Power_C_03}}
 {id=WavePlayer_19, properties {soundWaveAssetPtr=&/Katana_Swing_Power_C_02}}
 {id=WavePlayer_2, properties {soundWaveAssetPtr=&/Katana_Attack_Swing_Second_A_03}}
 {id=WavePlayer_22, properties {soundWaveAssetPtr=&/Katana_Swing_Power_A_01}}
 {id=WavePlayer_23, properties {soundWaveAssetPtr=&/Katana_Swing_Power_A_02}}
 {id=WavePlayer_24, properties {soundWaveAssetPtr=&/Katana_Swing_Power_D_01}}
 {id=WavePlayer_25, properties {soundWaveAssetPtr=&/Katana_Swing_Power_D_02}}
 {id=WavePlayer_26, properties {soundWaveAssetPtr=&/Katana_Swing_Power_D_03}}
 {id=WavePlayer_27, properties {soundWaveAssetPtr=&/Katana_Swing_Power_D_04}}
 {id=WavePlayer_28, properties {soundWaveAssetPtr=&Katana_Attack_Swing_Second_A_01}}
 {id=WavePlayer_29, properties {soundWaveAssetPtr=&Katana_Attack_Swing_Second_A_02}}
 {id=WavePlayer_3, properties {soundWaveAssetPtr=&/Katana_Attack_Swing_Second_A_04}}
 {id=WavePlayer_30, properties {soundWaveAssetPtr=&/Katana_Attack_Swing_01}}
 {id=WavePlayer_31, properties {soundWaveAssetPtr=&/Katana_Attack_Swing_02}}
 {id=WavePlayer_32, properties {soundWaveAssetPtr=&Katana_Swing_Power_C_03}}
 {id=WavePlayer_33, properties {soundWaveAssetPtr=&Katana_Swing_Power_C_02}}
 {id=WavePlayer_36, properties {soundWaveAssetPtr=&Katana_Swing_Power_D_01}}
 {id=WavePlayer_37, properties {soundWaveAssetPtr=&Katana_Swing_Power_D_02}}
 {id=WavePlayer_4, properties {soundWaveAssetPtr=&Katana_Attack_Swing_01}}
 {id=WavePlayer_5, properties {soundWaveAssetPtr=&Katana_Attack_Swing_02}}
 {id=WavePlayer_6, properties {soundWaveAssetPtr=&/Katana_Attack_Swing_03}}
 {id=WavePlayer_7, properties {soundWaveAssetPtr=&/Katana_Attack_Swing_04}}
]
properties {
 attenuationSettings=/Game/Sounds/Attenuation/Fort_Weapon_Melee_Attack_Close_Range_Stereo_Att
 maxDistance=10058.75
 soundClassObject=&Fort_WeaponMelee_PSM
 volumeMultiplier=0.9
 concurrencySet [&Player_Melee_Weapon_Conc]
}
