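# Sound cue for the impact of a dropped katana.
# A random node picks one of three equally weighted impact waves, which passes through a modulator for variation and a short fixed delay before playing.
# The cue uses a slightly raised pitch, shared weapon-drop attenuation and the loot-collect sound class.
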SoundCue /KatanaGameplay/Items/Katana/Sounds/Katana_DropImpact_Cue
root=Delay_0
nodes [
 {id=Delay_0, children [Modulator_0], properties {delayMax=0.1, delayMin=0.1}}
 {id=Modulator_0, children [Random_1]}
 {id=Random_1, children [WavePlayer_1,WavePlayer_3,WavePlayer_4], properties {weights [1,1,1]}}
 {id=WavePlayer_1, properties {soundWaveAssetPtr=&/Katana_DropImpact_01}}
 {id=WavePlayer_3, properties {soundWaveAssetPtr=&/Katana_DropImpact_02}}
 {id=WavePlayer_4, properties {soundWaveAssetPtr=&/Katana_DropImpact_03}}
]
properties {
 attenuationSettings=&Fort_Weapon_Drop_Att
 maxDistance=3000
 pitchMultiplier=1.1
 soundClassObject=&Fort_Loot_Collect
 concurrencySet [null]
}
