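# Sound cue that plays when a katana item is dropped in the world.
# A random node picks one of three drop waves with equal weight, then passes it through a modulator for variation.
# The cue applies slight pitch-up and volume reduction, uses the shared weapon-drop attenuation, and routes to the loot-collect sound class.
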SoundCue /KatanaGameplay/Items/Katana/Sounds/Katana_Drop_Cue
root=Modulator_0
nodes [
 {id=Modulator_0, children [Random_0]}
 {id=Random_0, children [WavePlayer_0,WavePlayer_2,WavePlayer_5], properties {weights [1,1,1]}}
 {id=WavePlayer_0, properties {soundWaveAssetPtr=&/Katana_DropB_01}}
 {id=WavePlayer_2, properties {soundWaveAssetPtr=&/Katana_DropB_02}}
 {id=WavePlayer_5, properties {soundWaveAssetPtr=&/Katana_DropB_03}}
]
properties {
 attenuationSettings=&Fort_Weapon_Drop_Att
 maxDistance=3000
 pitchMultiplier=1.1
 soundClassObject=&Fort_Loot_Collect
 volumeMultiplier=0.8
 concurrencySet [null]
}
