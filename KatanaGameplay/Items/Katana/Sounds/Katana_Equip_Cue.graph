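# Sound cue for equipping the katana, a close-range melee weapon sound capped at two concurrent plays.
# Mixes three randomized layers: equip and reach sounds with short staggered delays and random modulation, plus an unmodulated grab sound.
# A quality-level root sends its first two slots to a mix with more wave variations and its last two to a reduced-variation copy.
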SoundCue /KatanaGameplay/Items/Katana/Sounds/Katana_Equip_Cue
root=QualityLevel_0
nodes [
 {id=Delay_0, children [Random_1]}
 {id=Delay_1, children [Modulator_0], properties {delayMax=0.2, delayMin=0.2}}
 {id=Delay_2, children [Random_4]}
 {id=Delay_3, children [Random_0], properties {delayMax=0.03, delayMin=0.03}}
 {id=Delay_4, children [Modulator_3], properties {delayMax=0.2, delayMin=0.2}}
 {id=Delay_5, children [Random_3], properties {delayMax=0.03, delayMin=0.03}}
 {id=Mixer_0, children [Modulator_1,Delay_0,Delay_1], properties {inputVolume [1.8,1,1.2]}}
 {id=Mixer_1, children [Modulator_2,Delay_2,Delay_4], properties {inputVolume [1.8,1,1.2]}}
 {id=Modulator_0, children [Random_2], properties {pitchMax=1.1, pitchMin=0.9, volumeMax=1.1, volumeMin=0.9}}
 {id=Modulator_1, children [Delay_3], properties {pitchMax=1.1, pitchMin=1}}
 {id=Modulator_2, children [Delay_5], properties {pitchMax=1.1, pitchMin=1}}
 {id=Modulator_3, children [Random_5], properties {pitchMax=1.1, pitchMin=0.9, volumeMax=1.1, volumeMin=0.9}}
 {id=QualityLevel_0, children [Mixer_0,Mixer_0,Mixer_1,Mixer_1]}
 {id=Random_0, children [WavePlayer_0,WavePlayer_1,WavePlayer_2], properties {weights [1,1,1]}}
 {id=Random_1, children [WavePlayer_10, WavePlayer_11, WavePlayer_12], properties {weights [1,1,1]}}
 {id=Random_2, children [WavePlayer_17, WavePlayer_18], properties {weights [1,1]}}
 {id=Random_3, children [WavePlayer_3,WavePlayer_4], properties {bShouldExcludeFromBranchCulling=true, weights [1,1]}}
 {id=Random_4, children [WavePlayer_6,WavePlayer_7], properties {bShouldExcludeFromBranchCulling=true, weights [1,1]}}
 {id=Random_5, children [WavePlayer_9], properties {weights [1]}}
 {id=WavePlayer_0, properties {soundWaveAssetPtr=&/Katana_Equip_01}}
 {id=WavePlayer_1, properties {soundWaveAssetPtr=&/Katana_Equip_02}}
 {id=WavePlayer_10, properties {soundWaveAssetPtr=&/Katana_Equip_Grab_01}}
 {id=WavePlayer_11, properties {soundWaveAssetPtr=&/Katana_Equip_Grab_02}}
 {id=WavePlayer_12, properties {soundWaveAssetPtr=&/Katana_Equip_Grab_03}}
 {id=WavePlayer_17, properties {soundWaveAssetPtr=&/Katana_Equip_Reach_03}}
 {id=WavePlayer_18, properties {soundWaveAssetPtr=&/Katana_Equip_Reach_04}}
 {id=WavePlayer_2, properties {soundWaveAssetPtr=&/Katana_Equip_03}}
 {id=WavePlayer_3, properties {soundWaveAssetPtr=&Katana_Equip_01}}
 {id=WavePlayer_4, properties {soundWaveAssetPtr=&Katana_Equip_02}}
 {id=WavePlayer_6, properties {soundWaveAssetPtr=&Katana_Equip_Grab_01}}
 {id=WavePlayer_7, properties {soundWaveAssetPtr=&Katana_Equip_Grab_02}}
 {id=WavePlayer_9, properties {soundWaveAssetPtr=&Katana_Equip_Reach_03}}
]
properties {
 attenuationSettings=/Game/Sounds/Attenuation/Fort_Prop_Close_Att
 bOverrideConcurrency=true
 maxDistance=1800
 soundClassObject=&Fort_WeaponMelee_PSM
 volumeMultiplier=0.45
 concurrencyOverrides {maxCount=2}
}
