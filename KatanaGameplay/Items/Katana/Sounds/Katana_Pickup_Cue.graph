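# Sound cue for picking up the Katana weapon.
# A Random node picks one of three equally weighted pickup waves and feeds it into a Modulator at the root.
# It plays in the loot-collect sound class with the weapon pickup attenuation and slightly raised pitch and volume.
# A concurrency override allows one instance at a time and blocks new ones while it plays.
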SoundCue /KatanaGameplay/Items/Katana/Sounds/Katana_Pickup_Cue
root=Modulator_0
nodes [
 {id=Modulator_0, children [Random_1]}
 {id=Random_1, children [WavePlayer_0,WavePlayer_3,WavePlayer_4], properties {weights [1,1,1]}}
 {id=WavePlayer_0, properties {soundWaveAssetPtr=&/Katana_Pickup_01}}
 {id=WavePlayer_3, properties {soundWaveAssetPtr=&/Katana_Pickup_02}}
 {id=WavePlayer_4, properties {soundWaveAssetPtr=&/Katana_Pickup_03}}
]
properties {
 attenuationSettings=&Fort_Weapon_Pickup_Att
 bOverrideConcurrency=true
 maxDistance=3500
 pitchMultiplier=1.1
 soundClassObject=&Fort_Loot_Collect
 volumeMultiplier=1.194
 concurrencyOverrides {maxCount=1, resolutionRule=PreventNew}
}
