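# Sound cue for stowing the katana, played as a close-range melee weapon sound.
# A Random node picks one of three stow wave variations with equal weight, then feeds a Modulator at the root for playback variation.
# It uses close prop attenuation and the melee weapon sound class at reduced volume, with concurrency limited to two instances.
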SoundCue /KatanaGameplay/Items/Katana/Sounds/Katana_Stow_Cue
root=Modulator_0
nodes [
 {id=Modulator_0, children [Random_0]}
 {id=Random_0, children [WavePlayer_0,WavePlayer_1,WavePlayer_3], properties {weights [1,1,1]}}
 {id=WavePlayer_0, properties {soundWaveAssetPtr=&/Katana_StowB_01}}
 {id=WavePlayer_1, properties {soundWaveAssetPtr=&/Katana_StowB_02}}
 {id=WavePlayer_3, properties {soundWaveAssetPtr=&/Katana_StowB_03}}
]
properties {
 attenuationSettings=/Game/Sounds/Attenuation/Fort_Prop_Close_Att
 maxDistance=1800
 soundClassObject=&Fort_WeaponMelee_PSM
 volumeMultiplier=0.4
 concurrencyOverrides {maxCount=2}
}
